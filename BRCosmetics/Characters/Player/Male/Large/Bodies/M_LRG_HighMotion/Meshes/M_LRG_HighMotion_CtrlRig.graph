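# Control Rig for the large male HighMotion player body that drives its deformation bones for armor plates, limbs and torso.
# Spherical pose readers on clavicle, arm, thigh and foot bones produce alphas weighting plate and calf offsets.
# PBIK and aim units fit arm helpers to the animated skeleton, while calf rotation drives knees.
# Torso delta alphas shift aimed torso helpers, and skydive and parachute flags adjust shoulders and backpack.
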ControlRigBlueprint /BRCosmetics/Characters/Player/Male/Large/Bodies/M_LRG_HighMotion/Meshes/M_LRG_HighMotion_CtrlRig
graphs [
 ContainedGraph {nodes [
   n1 = RigVMFunctionEntryNode()
   SequenceAggregate = &RigVMFunction_Sequence(ExecuteContext=@n1.ExecuteContext)
   SequenceAggregate_1 = &RigVMFunction_Sequence(ExecuteContext=@SequenceAggregate.B)
   SequenceAggregate_2 = &RigVMFunction_Sequence(ExecuteContext=@SequenceAggregate_1.B)
   SequenceAggregate_3 = &RigVMFunction_Sequence(ExecuteContext=@SequenceAggregate_2.B)
   SequenceAggregate_4 = &RigVMFunction_Sequence(ExecuteContext=@SequenceAggregate_3.B)
   RigVMFunctionReturnNode(
    A=@SequenceAggregate.A
    B=@SequenceAggregate_1.A
    C=@SequenceAggregate_2.A
    D=@SequenceAggregate_3.A
    E=@SequenceAggregate_4.A
    F=@SequenceAggregate_4.B
   )
  ]}
 ContainedGraph {nodes [
   n1 = RigVMFunctionEntryNode()
   SequenceAggregate_3 = &RigVMFunction_Sequence(ExecuteContext=@n1.ExecuteContext)
   SequenceAggregate_3_1 = &RigVMFunction_Sequence(ExecuteContext=@SequenceAggregate_3.B)
   SequenceAggregate_3_2 = &RigVMFunction_Sequence(ExecuteContext=@SequenceAggregate_3_1.B)
   RigVMFunctionReturnNode(A=@SequenceAggregate_3.A, B=@SequenceAggregate_3_1.A, C=@SequenceAggregate_3_2.A, D=@SequenceAggregate_3_2.B)
  ]}
 ContainedGraph {nodes [
   n1 = RigVMFunctionEntryNode()
   SequenceAggregate_5 = &RigVMFunction_Sequence(ExecuteContext=@n1.ExecuteContext)
   SequenceAggregate_5_1 = &RigVMFunction_Sequence(ExecuteContext=@SequenceAggregate_5.B)
   RigVMFunctionReturnNode(A=@SequenceAggregate_5.A, B=@SequenceAggregate_5_1.A, C=@SequenceAggregate_5_1.B)
  ]}
 ContainedGraph {nodes [
   n1 = RigVMFunctionEntryNode()
   SequenceAggregate_1 = &RigVMFunction_Sequence(ExecuteContext=@n1.ExecuteContext)
   SequenceAggregate_1_1 = &RigVMFunction_Sequence(ExecuteContext=@SequenceAggregate_1.B)
   SequenceAggregate_1_2 = &RigVMFunction_Sequence(ExecuteContext=@SequenceAggregate_1_1.B)
   SequenceAggregate_1_3 = &RigVMFunction_Sequence(ExecuteContext=@SequenceAggregate_1_2.B)
   RigVMFunctionReturnNode(A=@SequenceAggregate_1.A, B=@SequenceAggregate_1_1.A, C=@SequenceAggregate_1_2.A, D=@SequenceAggregate_1_3.A, E=@SequenceAggregate_1_3.B)
  ]}
 RigVMFunctionLibrary {links [], nodes []}
 RigVMModel {nodes [
   RigUnit_BeginExecution = &RigUnit_BeginExecution()
   SequenceAggregate = RigVMAggregateNode(ExecuteContext=@RigUnit_BeginExecution.ExecutePin)
   SequenceAggregate_5 = RigVMAggregateNode(ExecuteContext=@SequenceAggregate.B)
   n1 = &RigUnit_SphericalPoseReader(
    ActiveRegionScaleFactors=(PositiveWidth=9,NegativeWidth=9,PositiveHeight=1,NegativeHeight=9)
    ActiveRegionSize=0.05
    Debug=(bDrawDebug=false,DebugScale=25,DebugSegments=20,DebugThickness=0.25)
    DriverAxis=(X=1,Y=0,Z=0)
    DriverItem=(Type=Bone,Name="clavicle_l")
    FalloffRegionScaleFactors=(PositiveWidth=1,NegativeWidth=1,PositiveHeight=0.46,NegativeHeight=1)
    FalloffSize=0.45
    OptionalParentItem=(Type=Bone)
    RotationOffset=(X=47.5,Y=0,Z=-90)
    ExecutePin=@SequenceAggregate_5.A
   )
   > RigVMVariableNode(Value=@n1.OutputParam) {variable=spr_clavicle_l_up_alpha}
   > n2 = &RigUnit_SphericalPoseReader(
    ActiveRegionScaleFactors=(PositiveWidth=3.5,NegativeWidth=0.5,PositiveHeight=3,NegativeHeight=3)
    ActiveRegionSize=0.1
    Debug=(bDrawDebug=false,DebugScale=25,DebugSegments=20,DebugThickness=0.25)
    DriverAxis=(X=-1,Y=0,Z=0)
    DriverItem=(Type=Bone,Name="thigh_l")
    FalloffRegionScaleFactors=(PositiveWidth=1,NegativeWidth=1,PositiveHeight=1,NegativeHeight=1)
    FalloffSize=0.45
    OptionalParentItem=(Type=Bone)
    RotationOffset=(X=90,Y=0,Z=0)
   )
   > RigVMVariableNode(Value=@n2.OutputParam) {variable=spr_thigh_l_up_alpha}
   > n3 = &RigUnit_SphericalPoseReader(
    ActiveRegionScaleFactors=(PositiveWidth=1,NegativeWidth=1,PositiveHeight=1,NegativeHeight=1)
    ActiveRegionSize=0.1
    Debug=(bDrawDebug=false,bDrawLocalAxes=false,DebugScale=25,DebugSegments=20,DebugThickness=0.25)
    DriverAxis=(X=0,Y=1,Z=0)
    DriverItem=(Type=Bone,Name="thigh_l")
    FalloffRegionScaleFactors=(PositiveWidth=1,NegativeWidth=1,PositiveHeight=1,NegativeHeight=1)
    FalloffSize=0.425
    OptionalParentItem=(Type=Bone)
    RotationOffset=(X=180,Y=0,Z=0)
   )
   > RigVMVariableNode(Value=@n3.OutputParam) {variable=spr_thigh_l_twist_out_alpha}
   > n4 = &RigUnit_SphericalPoseReader(
    ActiveRegionScaleFactors=(PositiveWidth=2,NegativeWidth=2,PositiveHeight=2,NegativeHeight=0.8)
    ActiveRegionSize=0.1
    Debug=(bDrawDebug=false,DebugScale=25,DebugSegments=20,DebugThickness=0.25)
    DriverAxis=(X=0,Y=1,Z=0)
    DriverItem=(Type=Bone,Name="foot_l")
    FalloffRegionScaleFactors=(PositiveWidth=1,NegativeWidth=1,PositiveHeight=1,NegativeHeight=0.65)
    FalloffSize=0.2
    OptionalParentItem=(Type=Bone)
    RotationOffset=(X=120,Y=0,Z=0)
   )
   > RigVMVariableNode(Value=@n4.OutputParam) {variable=spr_foot_l_rotate_out_alpha}
   > n5 = &RigUnit_SphericalPoseReader(
    ActiveRegionScaleFactors=(PositiveWidth=2,NegativeWidth=2,PositiveHeight=1.65,NegativeHeight=2)
    ActiveRegionSize=0.1
    Debug=(bDrawDebug=false,DebugScale=25,DebugSegments=20,DebugThickness=0.25)
    DriverAxis=(X=0,Y=1,Z=0)
    DriverItem=(Type=Bone,Name="foot_l")
    FalloffRegionScaleFactors=(PositiveWidth=1,NegativeWidth=1,PositiveHeight=0.8,NegativeHeight=1)
    FalloffSize=0.2
    OptionalParentItem=(Type=Bone)
    RotationOffset=(X=55,Y=0,Z=0)
   )
   > RigVMVariableNode(Value=@n5.OutputParam) {variable=spr_foot_l_rotate_in_alpha}
   > n6 = &RigUnit_SphericalPoseReader(
    ActiveRegionScaleFactors=(PositiveWidth=0.1,NegativeWidth=2,PositiveHeight=2,NegativeHeight=2)
    ActiveRegionSize=0.1
    Debug=(bDrawDebug=false,DebugScale=25,DebugSegments=20,DebugThickness=0.25)
    DriverAxis=(X=0,Y=1,Z=0)
    DriverItem=(Type=Bone,Name="foot_l")
    FalloffRegionScaleFactors=(PositiveWidth=0.5,NegativeWidth=1,PositiveHeight=0.5,NegativeHeight=0.5)
    FalloffSize=0.2
    FlipHeightScaling=false
    FlipWidthScaling=true
    OptionalParentItem=(Type=Bone)
    RotationOffset=(X=90,Y=0,Z=-25)
   )
   > RigVMVariableNode(Value=@n6.OutputParam) {variable=spr_foot_l_rotate_up_alpha}
   n7 = &RigUnit_SphericalPoseReader(
    ActiveRegionScaleFactors=(PositiveWidth=9,NegativeWidth=9,PositiveHeight=1,NegativeHeight=9)
    ActiveRegionSize=0.05
    Debug=(bDrawDebug=false,DebugScale=25,DebugSegments=20,DebugThickness=0.25)
    DriverAxis=(X=-1,Y=0,Z=0)
    DriverItem=(Type=Bone,Name="clavicle_r")
    FalloffRegionScaleFactors=(PositiveWidth=1,NegativeWidth=1,PositiveHeight=0.46,NegativeHeight=1)
    FalloffSize=0.45
    FlipHeightScaling=true
    FlipWidthScaling=true
    OptionalParentItem=(Type=Bone)
    RotationOffset=(X=132.5,Y=0,Z=90)
    ExecutePin=@SequenceAggregate_5.B
   )
   > RigVMVariableNode(Value=@n7.OutputParam) {variable=spr_clavicle_r_up_alpha}
   > n8 = &RigUnit_SphericalPoseReader(
    ActiveRegionScaleFactors=(PositiveWidth=4.5,NegativeWidth=1.5,PositiveHeight=4.5,NegativeHeight=4.5)
    ActiveRegionSize=0.1
    Debug=(bDrawDebug=false,bDrawLocalAxes=true,DebugScale=25,DebugSegments=20,DebugThickness=0.25)
    DriverAxis=(X=-1,Y=0,Z=0)
    DriverItem=(Type=Bone,Name="upperarm_r")
    FalloffRegionScaleFactors=(PositiveWidth=1,NegativeWidth=1,PositiveHeight=1,NegativeHeight=1)
    FalloffSize=0.45
    OptionalParentItem=(Type=Bone)
    RotationOffset=(X=-80,Y=10,Z=20)
   )
   > RigVMVariableNode(Value=@n8.OutputParam) {variable=spr_upperarm_r_rot_in_alpha}
   > n9 = &RigUnit_SphericalPoseReader(
    ActiveRegionScaleFactors=(PositiveWidth=3,NegativeWidth=0.5,PositiveHeight=1,NegativeHeight=3)
    ActiveRegionSize=0.1
    Debug=(bDrawDebug=false,bDrawLocalAxes=false,DebugScale=25,DebugSegments=20,DebugThickness=0.25)
    DriverAxis=(X=-1,Y=0,Z=0)
    DriverItem=(Type=Bone,Name="upperarm_r")
    FalloffRegionScaleFactors=(PositiveWidth=1,NegativeWidth=1,PositiveHeight=1,NegativeHeight=1)
    FalloffSize=0.3
    OptionalParentItem=(Type=Bone)
    RotationOffset=(X=-137.5,Y=43.5,Z=0)
   )
   > RigVMVariableNode(Value=@n9.OutputParam) {variable=spr_upperarm_r_ads_up_alpha}
   > n10 = &RigUnit_SphericalPoseReader(
    ActiveRegionScaleFactors=(PositiveWidth=3.5,NegativeWidth=0.5,PositiveHeight=3,NegativeHeight=3)
    ActiveRegionSize=0.1
    Debug=(bDrawDebug=false,DebugScale=25,DebugSegments=20,DebugThickness=0.25)
    DriverAxis=(X=1,Y=0,Z=0)
    DriverItem=(Type=Bone,Name="thigh_r")
    FalloffRegionScaleFactors=(PositiveWidth=1,NegativeWidth=1,PositiveHeight=1,NegativeHeight=1)
    FalloffSize=0.45
    FlipHeightScaling=false
    FlipWidthScaling=false
    OptionalParentItem=(Type=Bone)
    RotationOffset=(X=-90,Y=0,Z=0)
   )
   > RigVMVariableNode(Value=@n10.OutputParam) {variable=spr_thigh_r_up_alpha}
   > n11 = &RigUnit_SphericalPoseReader(
    ActiveRegionScaleFactors=(PositiveWidth=1,NegativeWidth=1,PositiveHeight=1,NegativeHeight=1)
    ActiveRegionSize=0.1
    Debug=(bDrawDebug=false,bDrawLocalAxes=false,DebugScale=25,DebugSegments=20,DebugThickness=0.25)
    DriverAxis=(X=0,Y=-1,Z=0)
    DriverItem=(Type=Bone,Name="thigh_r")
    FalloffRegionScaleFactors=(PositiveWidth=1,NegativeWidth=1,PositiveHeight=1,NegativeHeight=1)
    FalloffSize=0.425
    OptionalParentItem=(Type=Bone)
    RotationOffset=(X=0,Y=0,Z=0)
   )
   > RigVMVariableNode(Value=@n11.OutputParam) {variable=spr_thigh_r_twist_out_alpha}
   > n12 = &RigUnit_SphericalPoseReader(
    ActiveRegionScaleFactors=(PositiveWidth=2,NegativeWidth=2,PositiveHeight=2,NegativeHeight=0.8)
    ActiveRegionSize=0.1
    Debug=(bDrawDebug=false,DebugScale=25,DebugSegments=20,DebugThickness=0.25)
    DriverAxis=(X=0,Y=-1,Z=0)
    DriverItem=(Type=Bone,Name="foot_r")
    FalloffRegionScaleFactors=(PositiveWidth=1,NegativeWidth=1,PositiveHeight=1,NegativeHeight=0.65)
    FalloffSize=0.2
    OptionalParentItem=(Type=Bone)
    RotationOffset=(X=-60,Y=0,Z=0)
   )
   > RigVMVariableNode(Value=@n12.OutputParam) {variable=spr_foot_r_rotate_out_alpha}
   > n13 = &RigUnit_SphericalPoseReader(
    ActiveRegionScaleFactors=(PositiveWidth=2,NegativeWidth=2,PositiveHeight=1.65,NegativeHeight=2)
    ActiveRegionSize=0.1
    Debug=(bDrawDebug=false,DebugScale=25,DebugSegments=20,DebugThickness=0.25)
    DriverAxis=(X=0,Y=-1,Z=0)
    DriverItem=(Type=Bone,Name="foot_r")
    FalloffRegionScaleFactors=(PositiveWidth=1,NegativeWidth=1,PositiveHeight=0.8,NegativeHeight=1)
    FalloffSize=0.2
    OptionalParentItem=(Type=Bone)
    RotationOffset=(X=-125,Y=0,Z=0)
   )
   > RigVMVariableNode(Value=@n13.OutputParam) {variable=spr_foot_r_rotate_in_alpha}
   > n14 = &RigUnit_SphericalPoseReader(
    ActiveRegionScaleFactors=(PositiveWidth=0.1,NegativeWidth=2,PositiveHeight=2,NegativeHeight=2)
    ActiveRegionSize=0.1
    Debug=(bDrawDebug=false,DebugScale=25,DebugSegments=20,DebugThickness=0.25)
    DriverAxis=(X=0,Y=-1,Z=0)
    DriverItem=(Type=Bone,Name="foot_r")
    FalloffRegionScaleFactors=(PositiveWidth=0.5,NegativeWidth=1,PositiveHeight=0.5,NegativeHeight=0.5)
    FalloffSize=0.2
    OptionalParentItem=(Type=Bone)
    RotationOffset=(X=-90,Y=0,Z=-25)
   )
   > RigVMVariableNode(Value=@n14.OutputParam) {variable=spr_foot_r_rotate_up_alpha}
   SequenceAggregate_4 = &RigVMFunction_Sequence(ExecuteContext=@SequenceAggregate.C)
   n15 = &RigVMFunction_MathBoolToFloat(Value=RigVMVariableNode() {variable=get_is_skydiving}.Value)
   n16 = &RigVMFunction_MathDoubleSub(A=1, B=@n15.Result)
   &RigUnit_ModifyTransforms(
    ItemToModify=((Item=(Type=Bone,Name="dfrm_upperarm_plate_l"),Transform=(Rotation=(X=0,Y=0,Z=0,W=1),Translation=(X=0,Y=-3,Z=1.75),Scale3D=(X=1,Y=1,Z=1))))
    Weight=&RigVMFunction_MathDoubleMul(B=@n16.Result, A=RigVMVariableNode() {variable=spr_clavicle_l_up_alpha}.Value).Result
    ExecutePin=@SequenceAggregate_4.A
   )
   > &RigUnit_ModifyTransforms(
    ItemToModify=((Item=(Type=Bone,Name="dfrm_upperarm_plate_l"),Transform=(Rotation=(X=0,Y=0,Z=0,W=1),Translation=(X=0,Y=0,Z=-4),Scale3D=(X=1,Y=1,Z=1))))
    Weight=@n15.Result
   )
   &RigUnit_ModifyTransforms(
    ItemToModify=((Item=(Type=Bone,Name="dfrm_upperarm_plate_r"),Transform=(Rotation=(X=0,Y=0,Z=0,W=1),Translation=(X=0,Y=-3,Z=-2),Scale3D=(X=1,Y=1,Z=1))))
    Weight=&RigVMFunction_MathDoubleMul(B=@n16.Result, A=RigVMVariableNode() {variable=spr_clavicle_r_up_alpha}.Value).Result
    ExecutePin=@SequenceAggregate_4.B
   )
   n17 = RigVMVariableNode() {variable=spr_clavicle_r_up_alpha}
   Subtract = &RigVMFunction_MathDoubleSub(A=@n17.Value, B=@n15.Result)
   > &RigUnit_ModifyTransforms(
    ItemToModify=((Item=(Type=Bone,Name="dfrm_clavicle_r"),Transform=(Rotation=(X=0,Y=0,Z=0,W=1),Translation=(X=0,Y=0,Z=0),Scale3D=(X=1,Y=1,Z=1))))
    ItemToModify.0.Transform.Rotation=&RigVMFunction_MathQuaternionFromEuler(Euler=(X=0,Y=15,Z=0)).Result
    Weight=@Subtract.Result
   )
   > &RigUnit_ModifyTransforms(
    ItemToModify=((Item=(Type=Bone,Name="dfrm_upperarm_plate_r"),Transform=(Rotation=(X=0,Y=0,Z=0,W=1),Translation=(X=0,Y=-1.5,Z=0),Scale3D=(X=1,Y=1,Z=1))))
    Weight=@n17.Value
   )
   Multiply = &RigVMFunction_MathDoubleMul(B=RigVMVariableNode() {variable=spr_upperarm_r_ads_up_alpha}.Value, A=@n17.Value)
   n18 = &RigVMFunction_MathDoubleSub(A=@Multiply.Result, B=@n15.Result)
   > &RigUnit_ModifyTransforms(
    ItemToModify=((Item=(Type=Bone,Name="dfrm_upperarm_plate_r"),Transform=(Rotation=(X=0,Y=0,Z=0,W=1),Translation=(X=0,Y=-1.5,Z=-1.5),Scale3D=(X=1,Y=1,Z=1))))
    Weight=@n18.Result
   )
   > &RigUnit_ModifyTransforms(
    ItemToModify=((Item=(Type=Bone,Name="dfrm_scale_upperarm_plate_r"),Transform=(Rotation=(X=0,Y=0,Z=0,W=1),Translation=(X=0,Y=0,Z=0),Scale3D=(X=1,Y=0.9,Z=0.9))))
    Weight=@n18.Result
   )
   > &RigUnit_ModifyTransforms(
    ItemToModify=((Item=(Type=Bone,Name="dfrm_upperarm_plate_r"),Transform=(Rotation=(X=0,Y=0,Z=0,W=1),Translation=(X=0,Y=0,Z=4),Scale3D=(X=1,Y=1,Z=1))))
    Weight=@n15.Result
   )
   SequenceAggregate_2 = &RigVMFunction_Sequence(ExecuteContext=@SequenceAggregate.D)
   &RigUnit_SetRotation(
    Item=(Type=Bone,Name="dfrm_upperarm_l")
    Space=GlobalSpace
    Weight=1
    bInitial=False
    Value=&RigUnit_GetTransform(Item=(Type=Bone,Name="upperarm_l")).Transform.Rotation
    ExecutePin=@SequenceAggregate_2.A
   )
   > &RigUnit_SetRotation(
    Item=(Type=Bone,Name="dfrm_lowerarm_l")
    Space=GlobalSpace
    Weight=1
    bInitial=False
    Value=&RigUnit_GetTransform(Item=(Type=Bone,Name="lowerarm_l")).Transform.Rotation
   )
   RigUnit_GetTransform = &RigUnit_GetTransform(Item=(Type=Bone,Name="hand_l"))
   > /Script/PBIK.RigUnit_PBIK(
    BoneSettings=((Bone="dfrm_lowerarm_l",X=Locked,Y=Locked,Z=Free,PreferredAngles=(X=0,Y=0,Z=0)))
    Debug=(DrawScale=1)
    Effectors=((Bone="dfrm_hand_l",Transform=(Rotation=(X=0,Y=0,Z=0,W=1),Translation=(X=0,Y=0,Z=0),Scale3D=(X=1,Y=1,Z=1)),PositionAlpha=1,RotationAlpha=1,StrengthAlpha=1,PinRotation=1,ChainDepth=0))
    ExcludedBones=()
    Root=dfrm_upperarm_plate_l
    Settings="(Iterations=20,SubIterations=0,MassMultiplier=1,bAllowStretch=False,RootBehavior=PinToInput,PrePullRootSettings=(RotationAlpha=0,RotationAlphaX=1,RotationAlphaY=1,RotationAlphaZ=1,PositionAlpha=1,PositionAlphaX=1,PositionAlphaY=1,PositionAlphaZ=1),GlobalPu ...(truncated)"
    Effectors.0.Transform=@RigUnit_GetTransform.Transform
   ) {is=struct RigVMUnitNode}
   > /Script/ControlRig.RigUnit_AimItem(
    DebugSettings=(Scale=10,WorldOffset=(Rotation=(X=0,Y=0,Z=0,W=1),Translation=(X=0,Y=0,Z=0),Scale3D=(X=1,Y=1,Z=1)))
    Item=(Type=Bone,Name="dfrm_elbow_l")
    Primary=(Weight=1,Axis=(X=1,Y=0,Z=0),Target=(X=0,Y=0,Z=0),Kind=Location,Space=(Type=Bone,Name="dfrm_elbow_aim_l"))
    Secondary=(Weight=1,Axis=(X=0,Y=-1,Z=0),Target=(X=1,Y=0,Z=0),Kind=Location,Space=(Type=Bone,Name="dfrm_elbow_up_l"))
    Weight=1
   ) {is=struct RigVMUnitNode}
   RigUnit_SetRotation = &RigUnit_SetRotation(
    Item=(Type=Bone,Name="dfrm_upperarm_r")
    Space=GlobalSpace
    Weight=1
    bInitial=False
    bPropagateToChildren=true
    Value=&RigUnit_GetTransform(Item=(Type=Bone,Name="upperarm_r")).Transform.Rotation
    ExecutePin=@SequenceAggregate_2.B
   )
   > &RigUnit_SetRotation(
    Item=(Type=Bone,Name="dfrm_lowerarm_r")
    Space=GlobalSpace
    Weight=1
    bInitial=False
    bPropagateToChildren=true
    Value=&RigUnit_GetTransform(Item=(Type=Bone,Name="lowerarm_r")).Transform.Rotation
   )
   > PBIK = &RigUnit_PBIK(
    BoneSettings=((Bone="dfrm_lowerarm_r",X=Locked,Y=Locked,Z=Free,PreferredAngles=(X=0,Y=0,Z=0)))
    Debug=(DrawScale=1)
    Effectors=((Bone="dfrm_hand_r",Transform=(Rotation=(X=0,Y=0,Z=0,W=1),Translation=(X=0,Y=0,Z=0),Scale3D=(X=1,Y=1,Z=1)),PositionAlpha=1,RotationAlpha=1,StrengthAlpha=1,PinRotation=1,ChainDepth=0))
    ExcludedBones=()
    Root=dfrm_upperarm_plate_r
    Settings="(Iterations=20,SubIterations=0,MassMultiplier=1,bAllowStretch=False,RootBehavior=PinToInput,PrePullRootSettings=(RotationAlpha=0,RotationAlphaX=1,RotationAlphaY=1,RotationAlphaZ=1,PositionAlpha=1,PositionAlphaX=1,PositionAlphaY=1,PositionAlphaZ=1),GlobalPu ...(truncated)"
    Effectors.0.Transform=&RigUnit_GetTransform(Item=(Type=Bone,Name="hand_r")).Transform
   )
   > &RigUnit_AimItem(
    DebugSettings=(Scale=10,WorldOffset=(Rotation=(X=0,Y=0,Z=0,W=1),Translation=(X=0,Y=0,Z=0),Scale3D=(X=1,Y=1,Z=1)))
    Item=(Type=Bone,Name="dfrm_elbow_r")
    Primary=(Weight=1,Axis=(X=1,Y=0,Z=0),Target=(X=0,Y=0,Z=0),Kind=Location,Space=(Type=Bone,Name="dfrm_elbow_aim_r"))
    Secondary=(Weight=1,Axis=(X=0,Y=-1,Z=0),Target=(X=1,Y=0,Z=0),Kind=Location,Space=(Type=Bone,Name="dfrm_elbow_up_r"))
    Weight=1
   )
   RigVMFunction_Sequence = &RigVMFunction_Sequence(ExecuteContext=@SequenceAggregate.A)
   n19 = RigVMVariableNode() {variable=current_parachute_value}
   &RigUnit_RotationConstraintLocalSpaceOffset(
    AdvancedSettings=(RotationOrderForFilter=XZY)
    Child=(Type=Bone,Name="dfrm_clavicle_l")
    Filter=(bX=True,bY=True,bZ=True)
    Parents=((Item=(Type=Bone,Name="clavicle_l"),Weight=1))
    ExecutePin=@RigVMFunction_Sequence.B
    Weight=@n19.Value
   )
   > &RigUnit_RotationConstraintLocalSpaceOffset(
    AdvancedSettings=(RotationOrderForFilter=XZY)
    Child=(Type=Bone,Name="dfrm_clavicle_r")
    Filter=(bX=True,bY=True,bZ=True)
    Parents=((Item=(Type=Bone,Name="clavicle_r"),Weight=1))
    Weight=@n19.Value
   )
   RigVMFunction_ControlFlowBranch = &RigVMFunction_ControlFlowBranch(Condition=RigVMVariableNode() {variable=get_is_parachute_opened}.Value, ExecuteContext=@RigVMFunction_Sequence.A)
   RigVMVariableNode(Value=0.35, ExecuteContext=@RigVMFunction_ControlFlowBranch.False) {variable=current_parachute_value}
   RigVMVariableNode(Value=1, ExecuteContext=@RigVMFunction_ControlFlowBranch.True) {variable=current_parachute_value}
   SequenceAggregate_3 = RigVMAggregateNode(ExecuteContext=@SequenceAggregate.E)
   &RigUnit_ModifyTransforms(
    ItemToModify=((Item=(Type=Bone,Name="dfrm_hip_plate_l"),Transform=(Rotation=(X=0,Y=0,Z=0,W=1),Translation=(X=0,Y=-3,Z=0),Scale3D=(X=1,Y=1,Z=1))))
    ItemToModify.0.Transform.Rotation=&RigVMFunction_MathQuaternionFromEuler(Euler=(X=0,Y=0,Z=-8)).Result
    ExecutePin=@SequenceAggregate_3.A
    Weight=RigVMVariableNode() {variable=spr_thigh_l_up_alpha}.Value
   )
   > &RigUnit_ModifyTransforms(
    ItemToModify=((Item=(Type=Bone,Name="dfrm_hip_plate_l"),Transform=(Rotation=(X=0,Y=0,Z=0,W=1),Translation=(X=0,Y=0,Z=0),Scale3D=(X=1,Y=1,Z=1))))
    ItemToModify.0.Transform.Rotation=&RigVMFunction_MathQuaternionFromEuler(Euler=(X=0,Y=-10,Z=-50)).Result
    Weight=RigVMVariableNode() {variable=spr_thigh_l_twist_out_alpha}.Value
   )
   SequenceAggregate_6 = &RigVMFunction_Sequence(ExecuteContext=@SequenceAggregate_3.C)
   n20 = &RigVMFunction_MathQuaternionToEuler(RotationOrder=ZYX, Value=&RigUnit_GetTransform(Item=(Type=Bone,Name="calf_l"), Space=LocalSpace).Transform.Rotation)
   &RigUnit_ModifyTransforms(
    ItemToModify=((Item=(Type=Bone,Name="dfrm_knee_mstr_l"),Transform=(Rotation=(X=0,Y=0,Z=0,W=1),Translation=(X=0,Y=0,Z=0),Scale3D=(X=1,Y=1,Z=1))))
    Mode=OverrideLocal
    Weight=1
    ItemToModify.0.Transform.Rotation=&RigVMFunction_MathQuaternionFromEuler(
     Euler=(X=0,Y=0,Z=0)
     Euler.X=@n20.Result.X
     Euler.Y=@n20.Result.Y
     Euler.Z=&RigVMFunction_MathDoubleMul(B=-0.3, A=@n20.Result.Z).Result
    ).Result
    ExecutePin=@SequenceAggregate_6.A
   )
   n21 = &RigVMFunction_MathDoubleRemap(SourceMaximum=110, SourceMinimum=5, TargetMaximum=1, TargetMinimum=0, Value=@n20.Result.Z)
   > &RigUnit_ModifyTransforms(
    ItemToModify=((Item=(Type=Bone,Name="dfrm_knee_l"),Transform=(Rotation=(X=0,Y=0,Z=0,W=1),Translation=(X=4.5,Y=-3,Z=0),Scale3D=(X=1,Y=1,Z=1))))
    ItemToModify.0.Transform.Rotation=&RigVMFunction_MathQuaternionFromEuler(Euler=(X=0,Y=0,Z=-45)).Result
    Weight=@n21.Result
   )
   > &RigUnit_ModifyTransforms(
    ItemToModify=((Item=(Type=Bone,Name="dfrm_thigh_aim_l"),Transform=(Rotation=(X=0,Y=0,Z=0,W=1),Translation=(X=0,Y=4,Z=0),Scale3D=(X=1,Y=1,Z=1))))
    Weight=@n21.Result
   )
   > &RigUnit_AimItem(
    DebugSettings=(Scale=10,WorldOffset=(Rotation=(X=0,Y=0,Z=0,W=1),Translation=(X=0,Y=0,Z=0),Scale3D=(X=1,Y=1,Z=1)))
    Item=(Type=Bone,Name="dfrm_thigh_l")
    Primary=(Weight=1,Axis=(X=-1,Y=0,Z=0),Target=(X=0,Y=0,Z=0),Kind=Location,Space=(Type=Bone,Name="dfrm_thigh_aim_l"))
    Secondary=(Weight=1,Axis=(X=0,Y=1,Z=0),Target=(X=1,Y=0,Z=0),Kind=Location,Space=(Type=Bone,Name="dfrm_thigh_up_l"))
    Weight=1
   )
   n22 = &RigVMFunction_MathQuaternionToEuler(RotationOrder=ZYX, Value=&RigUnit_GetTransform(Item=(Type=Bone,Name="calf_r"), Space=LocalSpace).Transform.Rotation)
   &RigUnit_ModifyTransforms(
    ItemToModify=((Item=(Type=Bone,Name="dfrm_knee_mstr_r"),Transform=(Rotation=(X=0,Y=0,Z=0,W=1),Translation=(X=0,Y=0,Z=0),Scale3D=(X=1,Y=1,Z=1))))
    Mode=OverrideLocal
    Weight=1
    ItemToModify.0.Transform.Rotation=&RigVMFunction_MathQuaternionFromEuler(
     Euler=(X=0,Y=0,Z=0)
     Euler.X=@n22.Result.X
     Euler.Y=@n22.Result.Y
     Euler.Z=&RigVMFunction_MathDoubleMul(B=-0.3, A=@n22.Result.Z).Result
    ).Result
    ExecutePin=@SequenceAggregate_6.B
   )
   n23 = &RigVMFunction_MathDoubleRemap(SourceMaximum=110, SourceMinimum=5, TargetMaximum=1, TargetMinimum=0, Value=@n22.Result.Z)
   > &RigUnit_ModifyTransforms(
    ItemToModify=((Item=(Type=Bone,Name="dfrm_knee_r"),Transform=(Rotation=(X=0,Y=0,Z=0,W=1),Translation=(X=-4.5,Y=3,Z=0),Scale3D=(X=1,Y=1,Z=1))))
    ItemToModify.0.Transform.Rotation=&RigVMFunction_MathQuaternionFromEuler(Euler=(X=0,Y=0,Z=-45)).Result
    Weight=@n23.Result
   )
   > &RigUnit_ModifyTransforms(
    ItemToModify=((Item=(Type=Bone,Name="dfrm_thigh_aim_r"),Transform=(Rotation=(X=0,Y=0,Z=0,W=1),Translation=(X=0,Y=-4,Z=0),Scale3D=(X=1,Y=1,Z=1))))
    Weight=@n23.Result
   )
   > &RigUnit_AimItem(
    DebugSettings=(Scale=10,WorldOffset=(Rotation=(X=0,Y=0,Z=0,W=1),Translation=(X=0,Y=0,Z=0),Scale3D=(X=1,Y=1,Z=1)))
    Item=(Type=Bone,Name="dfrm_thigh_r")
    Primary=(Weight=1,Axis=(X=1,Y=0,Z=0),Target=(X=0,Y=0,Z=0),Kind=Location,Space=(Type=Bone,Name="dfrm_thigh_aim_r"))
    Secondary=(Weight=1,Axis=(X=0,Y=-1,Z=0),Target=(X=1,Y=0,Z=0),Kind=Location,Space=(Type=Bone,Name="dfrm_thigh_up_r"))
    Weight=1
   )
   &RigUnit_ModifyTransforms(
    ItemToModify=((Item=(Type=Bone,Name="dfrm_hip_plate_r"),Transform=(Rotation=(X=0,Y=0,Z=0,W=1),Translation=(X=0,Y=3,Z=0),Scale3D=(X=1,Y=1,Z=1))))
    ItemToModify.0.Transform.Rotation=&RigVMFunction_MathQuaternionFromEuler(Euler=(X=0,Y=0,Z=-8)).Result
    Weight=RigVMVariableNode() {variable=spr_thigh_r_up_alpha}.Value
    ExecutePin=@SequenceAggregate_3.B
   )
   > &RigUnit_ModifyTransforms(
    ItemToModify=((Item=(Type=Bone,Name="dfrm_hip_plate_r"),Transform=(Rotation=(X=0,Y=0,Z=0,W=1),Translation=(X=0,Y=0,Z=0),Scale3D=(X=1,Y=1,Z=1))))
    ItemToModify.0.Transform.Rotation=&RigVMFunction_MathQuaternionFromEuler(Euler=(X=0,Y=-10,Z=-50)).Result
    Weight=RigVMVariableNode() {variable=spr_thigh_r_twist_out_alpha}.Value
   )
   &RigUnit_ModifyTransforms(
    ItemToModify=((Item=(Type=Bone,Name="dfrm_calf_outer_l"),Transform=(Rotation=(X=0,Y=0,Z=0,W=1),Translation=(X=0,Y=0,Z=-3.5),Scale3D=(X=1,Y=1,Z=1))))
    ItemToModify.0.Transform.Rotation=&RigVMFunction_MathQuaternionFromEuler(Euler=(X=0,Y=15,Z=0)).Result
    Weight=RigVMVariableNode() {variable=spr_foot_l_rotate_out_alpha}.Value
    ExecutePin=@SequenceAggregate_3.D
   )
   > &RigUnit_ModifyTransforms(
    ItemToModify=((Item=(Type=Bone,Name="dfrm_calf_outer_l"),Transform=(Rotation=(X=0,Y=0,Z=0,W=1),Translation=(X=0,Y=0,Z=-1.5),Scale3D=(X=1,Y=1,Z=1))))
    ItemToModify.0.Transform.Rotation=&RigVMFunction_MathQuaternionFromEuler(Euler=(X=0,Y=-5,Z=0)).Result
    Weight=RigVMVariableNode() {variable=spr_foot_l_rotate_up_alpha}.Value
   )
   > &RigUnit_ModifyTransforms(
    ItemToModify=((Item=(Type=Bone,Name="dfrm_calf_inner_l"),Transform=(Rotation=(X=0,Y=0,Z=0,W=1),Translation=(X=0,Y=0,Z=2.5),Scale3D=(X=1,Y=1,Z=1))))
    ItemToModify.0.Transform.Rotation=&RigVMFunction_MathQuaternionFromEuler(Euler=(X=0,Y=-7.5,Z=0)).Result
    Weight=RigVMVariableNode() {variable=spr_foot_l_rotate_in_alpha}.Value
   )
   > &RigUnit_ModifyTransforms(
    ItemToModify=((Item=(Type=Bone,Name="dfrm_calf_outer_r"),Transform=(Rotation=(X=0,Y=0,Z=0,W=1),Translation=(X=0,Y=0,Z=3.5),Scale3D=(X=1,Y=1,Z=1))))
    ItemToModify.0.Transform.Rotation=&RigVMFunction_MathQuaternionFromEuler(Euler=(X=0,Y=15,Z=0)).Result
    Weight=RigVMVariableNode() {variable=spr_foot_r_rotate_out_alpha}.Value
   )
   > &RigUnit_ModifyTransforms(
    ItemToModify=((Item=(Type=Bone,Name="dfrm_calf_outer_r"),Transform=(Rotation=(X=0,Y=0,Z=0,W=1),Translation=(X=0,Y=0,Z=1.5),Scale3D=(X=1,Y=1,Z=1))))
    ItemToModify.0.Transform.Rotation=&RigVMFunction_MathQuaternionFromEuler(Euler=(X=0,Y=5,Z=0)).Result
    Weight=RigVMVariableNode() {variable=spr_foot_r_rotate_up_alpha}.Value
   )
   > &RigUnit_ModifyTransforms(
    ItemToModify=((Item=(Type=Bone,Name="dfrm_calf_inner_r"),Transform=(Rotation=(X=0,Y=0,Z=0,W=1),Translation=(X=0,Y=0,Z=-2.5),Scale3D=(X=1,Y=1,Z=1))))
    ItemToModify.0.Transform.Rotation=&RigVMFunction_MathQuaternionFromEuler(Euler=(X=0,Y=-7.5,Z=0)).Result
    Weight=RigVMVariableNode() {variable=spr_foot_r_rotate_in_alpha}.Value
   )
   SequenceAggregate_1 = RigVMAggregateNode(ExecuteContext=@SequenceAggregate.F)
   n24 = RigVMVariableNode() {variable=delta_ref_torso_twist_right_alpha}
   &RigUnit_ModifyTransforms(
    ItemToModify=((Item=(Type=Bone,Name="dfrm_mstr_front_torso_base_l"),Transform=(Rotation=(X=0,Y=0,Z=0,W=1),Translation=(X=0,Y=0,Z=1.5),Scale3D=(X=1,Y=1,Z=1))))
    ExecutePin=@SequenceAggregate_1.A
    Weight=@n24.Value
   )
   > &RigUnit_ModifyTransforms(
    ItemToModify=((Item=(Type=Bone,Name="dfrm_mstr_front_torso_end_l"),Transform=(Rotation=(X=0,Y=0,Z=0,W=1),Translation=(X=0,Y=1.5,Z=-1),Scale3D=(X=1,Y=1,Z=1))))
    Weight=@n24.Value
   )
   > &RigUnit_ModifyTransforms(
    ItemToModify=((Item=(Type=Bone,Name="dfrm_mstr_back_torso_base_l"),Transform=(Rotation=(X=0,Y=0,Z=0,W=1),Translation=(X=0,Y=-2,Z=-2),Scale3D=(X=1,Y=1,Z=1))))
    Weight=@n24.Value
   )
   > &RigUnit_ModifyTransforms(
    ItemToModify=((Item=(Type=Bone,Name="dfrm_mstr_back_torso_end_l"),Transform=(Rotation=(X=0,Y=0,Z=0,W=1),Translation=(X=0,Y=-2,Z=0.5),Scale3D=(X=1,Y=1,Z=1))))
    Weight=@n24.Value
   )
   > &RigUnit_ModifyTransforms(
    ItemToModify=((Item=(Type=Bone,Name="dfrm_mstr_front_torso_base_r"),Transform=(Rotation=(X=0,Y=0,Z=0,W=1),Translation=(X=0,Y=-0.5,Z=-0.5),Scale3D=(X=1,Y=1,Z=1))))
    Weight=@n24.Value
   )
   > &RigUnit_ModifyTransforms(
    ItemToModify=((Item=(Type=Bone,Name="dfrm_mstr_front_torso_end_r"),Transform=(Rotation=(X=0,Y=0,Z=0,W=1),Translation=(X=0,Y=-1.75,Z=-0.5),Scale3D=(X=1,Y=1,Z=1))))
    Weight=@n24.Value
   )
   > &RigUnit_ModifyTransforms(
    ItemToModify=((Item=(Type=Bone,Name="dfrm_mstr_back_torso_base_r"),Transform=(Rotation=(X=0,Y=0,Z=0,W=1),Translation=(X=0,Y=0,Z=1),Scale3D=(X=1,Y=1,Z=1))))
    Weight=@n24.Value
   )
   > &RigUnit_ModifyTransforms(
    ItemToModify=((Item=(Type=Bone,Name="dfrm_mstr_back_torso_end_r"),Transform=(Rotation=(X=0,Y=0,Z=0,W=1),Translation=(X=0,Y=-1,Z=-1),Scale3D=(X=1,Y=1,Z=1))))
    Weight=@n24.Value
   )
   &RigUnit_AimItem(
    DebugSettings=(Scale=10,WorldOffset=(Rotation=(X=0,Y=0,Z=0,W=1),Translation=(X=0,Y=0,Z=0),Scale3D=(X=1,Y=1,Z=1)))
    Item=(Type=Bone,Name="dfrm_front_torso_base_l")
    Primary=(Weight=1,Axis=(X=1,Y=0,Z=0),Target=(X=0,Y=0,Z=0),Kind=Location,Space=(Type=Bone,Name="dfrm_mstr_front_torso_end_l"))
    Secondary=(Weight=1,Axis=(X=0,Y=-1,Z=0),Target=(X=1,Y=0,Z=0),Kind=Location,Space=(Type=Bone,Name="dfrm_front_torso_base_up_l"))
    Weight=1
    ExecutePin=@SequenceAggregate_1.C
   )
   > &RigUnit_AimItem(
    DebugSettings=(Scale=10,WorldOffset=(Rotation=(X=0,Y=0,Z=0,W=1),Translation=(X=0,Y=0,Z=0),Scale3D=(X=1,Y=1,Z=1)))
    Item=(Type=Bone,Name="dfrm_front_torso_end_l")
    Primary=(Weight=1,Axis=(X=1,Y=0,Z=0),Target=(X=0,Y=0,Z=0),Kind=Location,Space=(Type=Bone,Name="dfrm_mstr_front_torso_base_l"))
    Secondary=(Weight=1,Axis=(X=0,Y=-1,Z=0),Target=(X=1,Y=0,Z=0),Kind=Location,Space=(Type=Bone,Name="dfrm_front_torso_base_up_l"))
    Weight=1
   )
   > &RigUnit_AimItem(
    DebugSettings=(Scale=10,WorldOffset=(Rotation=(X=0,Y=0,Z=0,W=1),Translation=(X=0,Y=0,Z=0),Scale3D=(X=1,Y=1,Z=1)))
    Item=(Type=Bone,Name="dfrm_back_torso_base_l")
    Primary=(Weight=1,Axis=(X=1,Y=0,Z=0),Target=(X=0,Y=0,Z=0),Kind=Location,Space=(Type=Bone,Name="dfrm_mstr_back_torso_end_l"))
    Secondary=(Weight=1,Axis=(X=0,Y=-1,Z=0),Target=(X=1,Y=0,Z=0),Kind=Location,Space=(Type=Bone,Name="dfrm_back_torso_base_up_l"))
    Weight=1
   )
   > &RigUnit_AimItem(
    DebugSettings=(Scale=10,WorldOffset=(Rotation=(X=0,Y=0,Z=0,W=1),Translation=(X=0,Y=0,Z=0),Scale3D=(X=1,Y=1,Z=1)))
    Item=(Type=Bone,Name="dfrm_back_torso_end_l")
    Primary=(Weight=1,Axis=(X=1,Y=0,Z=0),Target=(X=0,Y=0,Z=0),Kind=Location,Space=(Type=Bone,Name="dfrm_back_torso_mid_l"))
    Secondary=(Weight=1,Axis=(X=0,Y=-1,Z=0),Target=(X=1,Y=0,Z=0),Kind=Location,Space=(Type=Bone,Name="dfrm_back_torso_end_up_l"))
    Weight=1
   )
   &RigUnit_AimItem(
    DebugSettings=(Scale=10,WorldOffset=(Rotation=(X=0,Y=0,Z=0,W=1),Translation=(X=0,Y=0,Z=0),Scale3D=(X=1,Y=1,Z=1)))
    Item=(Type=Bone,Name="dfrm_front_torso_base_r")
    Primary=(Weight=1,Axis=(X=1,Y=0,Z=0),Target=(X=0,Y=0,Z=0),Kind=Location,Space=(Type=Bone,Name="dfrm_mstr_front_torso_end_r"))
    Secondary=(Weight=1,Axis=(X=0,Y=-1,Z=0),Target=(X=1,Y=0,Z=0),Kind=Location,Space=(Type=Bone,Name="dfrm_front_torso_base_up_r"))
    Weight=1
    ExecutePin=@SequenceAggregate_1.D
   )
   > &RigUnit_AimItem(
    DebugSettings=(Scale=10,WorldOffset=(Rotation=(X=0,Y=0,Z=0,W=1),Translation=(X=0,Y=0,Z=0),Scale3D=(X=1,Y=1,Z=1)))
    Item=(Type=Bone,Name="dfrm_front_torso_end_r")
    Primary=(Weight=1,Axis=(X=1,Y=0,Z=0),Target=(X=0,Y=0,Z=0),Kind=Location,Space=(Type=Bone,Name="dfrm_mstr_front_torso_base_r"))
    Secondary=(Weight=1,Axis=(X=0,Y=-1,Z=0),Target=(X=1,Y=0,Z=0),Kind=Location,Space=(Type=Bone,Name="dfrm_front_torso_end_up_r"))
    Weight=1
   )
   > &RigUnit_AimItem(
    DebugSettings=(Scale=10,WorldOffset=(Rotation=(X=0,Y=0,Z=0,W=1),Translation=(X=0,Y=0,Z=0),Scale3D=(X=1,Y=1,Z=1)))
    Item=(Type=Bone,Name="dfrm_back_torso_base_r")
    Primary=(Weight=1,Axis=(X=1,Y=0,Z=0),Target=(X=0,Y=0,Z=0),Kind=Location,Space=(Type=Bone,Name="dfrm_back_torso_end_r"))
    Secondary=(Weight=1,Axis=(X=0,Y=-1,Z=0),Target=(X=1,Y=0,Z=0),Kind=Location,Space=(Type=Bone,Name="dfrm_back_torso_base_up_r"))
    Weight=1
   )
   > &RigUnit_AimItem(
    DebugSettings=(Scale=10,WorldOffset=(Rotation=(X=0,Y=0,Z=0,W=1),Translation=(X=0,Y=0,Z=0),Scale3D=(X=1,Y=1,Z=1)))
    Item=(Type=Bone,Name="dfrm_back_torso_end_r")
    Primary=(Weight=1,Axis=(X=1,Y=0,Z=0),Target=(X=0,Y=0,Z=0),Kind=Location,Space=(Type=Bone,Name="dfrm_back_torso_mid_r"))
    Secondary=(Weight=1,Axis=(X=0,Y=-1,Z=0),Target=(X=1,Y=0,Z=0),Kind=Location,Space=(Type=Bone,Name="dfrm_back_torso_end_up_r"))
    Weight=1
   )
   Cast = &RigVMFunction_MathBoolToFloat(Value=RigVMVariableNode() {variable=get_is_skydive_floating}.Value)
   AlphaInterp = &RigVMFunction_AlphaInterp(InterpSpeedDecreasing=10, InterpSpeedIncreasing=10, Scale=1, bClampResult=False, Value=@Cast.Result)
   &RigUnit_ModifyTransforms(
    ItemToModify=((Item=(Type=Bone,Name="attach_backpack"),Transform=(Rotation=(X=0,Y=0,Z=0,W=1),Translation=(X=0,Y=-4,Z=0),Scale3D=(X=1,Y=1,Z=1))))
    Weight=@AlphaInterp.Result
    ExecutePin=@SequenceAggregate_1.E
   )
   > &RigUnit_ParentConstraint(
    AdvancedSettings=(RotationOrderForFilter=XZY)
    Child=(Type=Bone,Name="weapon_lowerarm_r")
    Parents=((Item=(Type=Bone,Name="dfrm_lowerarm_r"),Weight=1))
    Weight=1
   )
   n25 = RigVMVariableNode() {variable=delta_ref_torso_rotate_forward_alpha}
   &RigUnit_ModifyTransforms(
    ItemToModify=((Item=(Type=Bone,Name="dfrm_back_torso_mid_l"),Transform=(Rotation=(X=0,Y=0,Z=0,W=1),Translation=(X=3,Y=-2,Z=0),Scale3D=(X=1,Y=1,Z=1))))
    Weight=@n25.Value
    ExecutePin=@SequenceAggregate_1.B
   )
   > &RigUnit_ModifyTransforms(
    ItemToModify=((Item=(Type=Bone,Name="dfrm_back_torso_mid_r"),Transform=(Rotation=(X=0,Y=0,Z=0,W=1),Translation=(X=3,Y=-2,Z=0),Scale3D=(X=1,Y=1,Z=1))))
    Weight=@n25.Value
   )
   > &RigUnit_ModifyTransforms(
    ItemToModify=((Item=(Type=Bone,Name="dfrm_mstr_back_torso_end_l"),Transform=(Rotation=(X=0,Y=0,Z=0,W=1),Translation=(X=0,Y=-0.5,Z=1),Scale3D=(X=1,Y=1,Z=1))))
    Weight=@n25.Value
   )
   > &RigUnit_ModifyTransforms(
    ItemToModify=((Item=(Type=Bone,Name="dfrm_mstr_back_torso_end_r"),Transform=(Rotation=(X=0,Y=0,Z=0,W=1),Translation=(X=0,Y=-0.5,Z=-1),Scale3D=(X=1,Y=1,Z=1))))
    Weight=@n25.Value
   )
  ]}
]
hierarchy [
 root {loc=<0 0 0>, rot=<0 0 0 1>}
 pelvis {parent=root, loc=<0 -0.7321 92.4218>, rot=<0.0127 0.707 -0.0127 -0.707>}
 spine_01 {parent=pelvis, loc=<5.6002 -0.5128 0>, rot=<0 0 0.034 -0.9994>}
 spine_02 {parent=spine_01, loc=<8.0753 0.0139 0>, rot=<0 0 0.0238 -0.9997>}
 spine_03 {parent=spine_02, loc=<7.0393 0 0>, rot=<0 0 0.0149 -0.9999>}
 spine_04 {parent=spine_03, loc=<8.1024 0 0>, rot=<0 0 -0.0132 -0.9999>}
 spine_05 {parent=spine_04, loc=<10.0146 0 0>, rot=<0 0 -0.0015 -1>}
 clavicle_l {parent=spine_05, loc=<2.5856 -0.2866 -4.3777>, rot=<0.0465 -0.7289 -0.009 -0.6829>}
 upperarm_l {parent=clavicle_l, loc=<18.2264 0.1342 -0.001>, rot=<0.0253 -0.2693 0.0013 -0.9627>}
 lowerarm_l {parent=upperarm_l, loc=<29.0071 0.1686 -1.239>, rot=<0.0391 -0.0099 -0.2453 -0.9686>}
 hand_l {parent=lowerarm_l, loc=<25.8607 0.222 1.0274>, rot=<-0.7193 -0.0988 0.1193 -0.6773>}
 thumb_01_l {parent=hand_l, loc=<1.3118 -0.3191 -2.459>, rot=<0.6546 -0.4042 0.0664 -0.6354>}
 thumb_02_l {parent=thumb_01_l, loc=<5.9336 -0.0448 0.0401>, rot=<-0.0037 0.0213 0.298 -0.9543>}
 thumb_03_l {parent=thumb_02_l, loc=<2.2223 0.3682 -0.0907>, rot=<0.0118 -0.013 0.0066 -0.9998>}
 index_metacarpal_l {parent=hand_l, loc=<1.6535 0.0183 -2.96>, rot=<0.0158 -0.0124 -0.0173 -0.9996>}
 index_01_l {parent=index_metacarpal_l, loc=<8.3008 0.0357 -0.0524>, rot=<-0.0044 0.0485 0.0924 -0.9945>}
 index_02_l {parent=index_01_l, loc=<3.3587 -0.3004 0.0452>, rot=<-0.002 0.0027 0.3225 -0.9466>}
 index_03_l {parent=index_02_l, loc=<2.5578 -0.123 -0.0252>, rot=<-0.002 0.0067 0.2558 -0.9667>}
 middle_metacarpal_l {parent=hand_l, loc=<2.1856 -0.1725 -0.5984>, rot=<0.006 0.0168 -0.0329 -0.9993>}
 middle_01_l {parent=middle_metacarpal_l, loc=<7.3275 -0.2447 0.0129>, rot=<-0.0208 0.0042 0.0858 -0.9961>}
 middle_02_l {parent=middle_01_l, loc=<4.0123 0.1401 0.1175>, rot=<-0.0218 0.0258 0.3512 -0.9357>}
 middle_03_l {parent=middle_02_l, loc=<2.6077 0.0534 0.1373>, rot=<0.0082 0.0188 0.1708 -0.9851>}
 ring_metacarpal_l {parent=hand_l, loc=<2.2074 -0.3625 1.3638>, rot=<-0.0284 0.0722 -0.0458 -0.9959>}
 ring_01_l {parent=ring_metacarpal_l, loc=<7.0295 -0.6197 -0.1446>, rot=<-0.0489 -0.0252 0.1892 -0.9804>}
 ring_02_l {parent=ring_01_l, loc=<3.6174 0.4647 -0.1027>, rot=<0 0 0.2806 -0.9598>}
 ring_03_l {parent=ring_02_l, loc=<2.4152 0.114 0.0199>, rot=<0 0 0.1189 -0.9929>}
 pinky_metacarpal_l {parent=hand_l, loc=<1.7616 -0.9753 3.3814>, rot=<-0.0397 0.0814 -0.069 -0.9935>}
 pinky_01_l {parent=pinky_metacarpal_l, loc=<7.1709 -0.4831 -0.1094>, rot=<-0.0654 -0.0035 0.2242 -0.9723>}
 pinky_02_l {parent=pinky_01_l, loc=<2.8274 0.0893 -0.1043>, rot=<0.0037 0.0011 0.2854 -0.9584>}
 pinky_03_l {parent=pinky_02_l, loc=<2.2046 0.0307 -0.0078>, rot=<-0.0002 0.0098 0.1445 -0.9895>}
 attach_cape {parent=spine_05, loc=<9.1509 -13.3783 0>, rot=<0.0616 -0.7044 -0.0616 -0.7044>}
 attach_backpack {parent=spine_05, loc=<-0.123 -15.1423 0>, rot=<-0.0093 -0.707 0.0093 -0.707>}
 neck_01 {parent=spine_05, loc=<10.815 1.9196 0>, rot=<0 0 -0.1864 -0.9825>}
 neck_02 {parent=neck_01, loc=<4.7347 0.3045 0>, rot=<0 0 0.0363 -0.9993>}
 head {parent=neck_02, loc=<5.7632 0.5552 0>, rot=<0 0 0.115 -0.9934>}
 clavicle_r {parent=spine_05, loc=<2.5855 -0.2866 4.3777>, rot=<0.7289 0.0465 -0.6829 0.009>}
 upperarm_r {parent=clavicle_r, loc=<-18.2264 -0.1342 0.001>, rot=<0.0253 -0.2693 0.0013 -0.9627>}
 lowerarm_r {parent=upperarm_r, loc=<-29.0071 -0.1686 1.239>, rot=<0.0391 -0.0099 -0.2453 -0.9686>}
 hand_r {parent=lowerarm_r, loc=<-25.8607 -0.222 -1.0274>, rot=<-0.7193 -0.0988 0.1193 -0.6773>}
 thumb_01_r {parent=hand_r, loc=<-1.3118 0.3191 2.459>, rot=<0.6546 -0.4042 0.0664 -0.6354>}
 thumb_02_r {parent=thumb_01_r, loc=<-5.9336 0.0448 -0.0401>, rot=<-0.0037 0.0213 0.298 -0.9543>}
 thumb_03_r {parent=thumb_02_r, loc=<-2.2223 -0.3682 0.0907>, rot=<0.0118 -0.013 0.0066 -0.9998>}
 index_metacarpal_r {parent=hand_r, loc=<-1.6535 -0.0183 2.96>, rot=<0.0158 -0.0124 -0.0173 -0.9996>}
 index_01_r {parent=index_metacarpal_r, loc=<-8.3008 -0.0357 0.0524>, rot=<-0.0044 0.0485 0.0924 -0.9945>}
 index_02_r {parent=index_01_r, loc=<-3.3587 0.3004 -0.0452>, rot=<-0.002 0.0027 0.3225 -0.9466>}
 index_03_r {parent=index_02_r, loc=<-2.5578 0.123 0.0252>, rot=<-0.002 0.0067 0.2558 -0.9667>}
 middle_metacarpal_r {parent=hand_r, loc=<-2.1856 0.1725 0.5984>, rot=<0.006 0.0168 -0.0329 -0.9993>}
 middle_01_r {parent=middle_metacarpal_r, loc=<-7.3275 0.2447 -0.0129>, rot=<-0.0208 0.0042 0.0858 -0.9961>}
 middle_02_r {parent=middle_01_r, loc=<-4.0123 -0.1401 -0.1175>, rot=<-0.0218 0.0258 0.3512 -0.9357>}
 middle_03_r {parent=middle_02_r, loc=<-2.6077 -0.0534 -0.1373>, rot=<0.0082 0.0188 0.1708 -0.9851>}
 ring_metacarpal_r {parent=hand_r, loc=<-2.2074 0.3625 -1.3638>, rot=<-0.0284 0.0722 -0.0458 -0.9959>}
 ring_01_r {parent=ring_metacarpal_r, loc=<-7.0295 0.6197 0.1446>, rot=<-0.0489 -0.0252 0.1892 -0.9804>}
 ring_02_r {parent=ring_01_r, loc=<-3.6174 -0.4647 0.1027>, rot=<0 0 0.2806 -0.9598>}
 ring_03_r {parent=ring_02_r, loc=<-2.4152 -0.114 -0.0199>, rot=<0 0 0.1189 -0.9929>}
 pinky_metacarpal_r {parent=hand_r, loc=<-1.7616 0.9753 -3.3814>, rot=<-0.0397 0.0814 -0.069 -0.9935>}
 pinky_01_r {parent=pinky_metacarpal_r, loc=<-7.1709 0.4831 0.1094>, rot=<-0.0654 -0.0035 0.2242 -0.9723>}
 pinky_02_r {parent=pinky_01_r, loc=<-2.8274 -0.0893 0.1043>, rot=<0.0037 0.0011 0.2854 -0.9584>}
 pinky_03_r {parent=pinky_02_r, loc=<-2.2046 -0.0307 0.0078>, rot=<-0.0002 0.0098 0.1445 -0.9895>}
 weapon_lowerarm_r {parent=lowerarm_r, loc=<0 0 0>, rot=<0 0 0 1>}
 dfrm_clavicle_l {parent=spine_05, loc=<2.5856 -0.2866 -4.3777>, rot=<-0.4639 0.4658 0.5686 0.4944>}
 dfrm_upperarm_plate_l {parent=dfrm_clavicle_l, loc=<13.734 0 0>, rot=<-0.0292 -0.0882 0.0137 -0.9956>}
 dfrm_upperarm_l {parent=dfrm_upperarm_plate_l, loc=<4.7086 -0.1168 -2.4596>, rot=<-0.6408 -0.2064 -0.1866 -0.7155>}
 dfrm_elbow_l {parent=dfrm_upperarm_l, loc=<24.5074 -0.8163 2.7226>, rot=<0.6935 0.0596 -0.1398 -0.7042>}
 dfrm_lowerarm_l {parent=dfrm_upperarm_l, loc=<29.0071 0.1686 -1.239>, rot=<0.0391 -0.0099 -0.2453 -0.9686>}
 dfrm_hand_l {parent=dfrm_lowerarm_l, loc=<25.8607 0.222 1.0274>, rot=<-0.7193 -0.0988 0.1193 -0.6773>}
 dfrm_elbow_aim_l {parent=dfrm_lowerarm_l, loc=<2.9289 -0.5487 3.1986>, rot=<0.6815 0.0174 0.0104 -0.7315>}
 dfrm_elbow_up_l {parent=dfrm_upperarm_l, loc=<25.2838 -0.7625 8.0304>, rot=<0.6935 0.0596 -0.1398 -0.7042>}
 dfrm_scale_upperarm_plate_l {parent=dfrm_upperarm_plate_l, loc=<0 0 0>, rot=<0 0 0 1>}
 dfrm_shoulder_housing_l {parent=dfrm_clavicle_l, loc=<13.734 0 0>, rot=<-0.0292 -0.0882 0.0137 -0.9956>}
 dfrm_clavicle_r {parent=spine_05, loc=<2.5855 -0.2866 4.3777>, rot=<0.4684 -0.4606 0.5729 0.4901>}
 dfrm_upperarm_plate_r {parent=dfrm_clavicle_r, loc=<13.734 0 0>, rot=<0.0414 0.0882 0.013 -0.9952>}
 dfrm_upperarm_r {parent=dfrm_upperarm_plate_r, loc=<4.7085 -0.1334 2.4588>, rot=<-0.2068 0.6387 0.7175 -0.1858>}
 dfrm_elbow_r {parent=dfrm_upperarm_r, loc=<-24.5073 0.8163 -2.7227>, rot=<-0.0596 0.6935 0.7042 -0.1398>}
 dfrm_elbow_up_r {parent=dfrm_upperarm_r, loc=<-25.2839 0.7625 -8.0301>, rot=<-0.0596 0.6935 0.7042 -0.1398>}
 dfrm_lowerarm_r {parent=dfrm_upperarm_r, loc=<-29.0071 -0.1686 1.239>, rot=<0.0391 -0.0099 -0.2453 -0.9686>}
 dfrm_elbow_aim_r {parent=dfrm_lowerarm_r, loc=<-2.9287 0.5486 -3.1989>, rot=<-0.0174 0.6815 0.7315 0.0104>}
 dfrm_hand_r {parent=dfrm_lowerarm_r, loc=<-25.8607 -0.222 -1.0274>, rot=<-0.7193 -0.0988 0.1193 -0.6773>}
 dfrm_scale_upperarm_plate_r {parent=dfrm_upperarm_plate_r, loc=<0 0 0>, rot=<0 0 0 1>}
 dfrm_shoulder_housing_r {parent=dfrm_clavicle_r, loc=<13.734 0 0>, rot=<0.0414 0.0882 0.013 -0.9952>}
 dfrm_mstr_front_torso_end_l {parent=spine_03, loc=<-1.1265 5.406 -8.492>, rot=<0.1163 0.0179 0.9921 0.0429>}
 dfrm_front_torso_end_l {parent=dfrm_mstr_front_torso_end_l, loc=<0 0 0>, rot=<0 0 0 1>}
 dfrm_front_torso_end_up_l {parent=dfrm_mstr_front_torso_end_l, loc=<0 -2 0>, rot=<0 0 0 1>, scale=<0.25 0.25 0.25>}
 dfrm_mstr_back_torso_end_l {parent=spine_03, loc=<1.8115 -3.6273 -7.7844>, rot=<0.1736 -0.9773 -0.0081 0.1215>}
 dfrm_back_torso_end_l {parent=dfrm_mstr_back_torso_end_l, loc=<0 0 0>, rot=<0 0 0 1>}
 dfrm_back_torso_end_up_l {parent=dfrm_mstr_back_torso_end_l, loc=<0 -2 0>, rot=<0 0 0 1>, scale=<0.25 0.25 0.25>}
 dfrm_mstr_front_torso_end_r {parent=spine_03, loc=<-1.1269 5.406 8.492>, rot=<-0.1163 -0.0179 0.9921 0.0429>}
 dfrm_front_torso_end_r {parent=dfrm_mstr_front_torso_end_r, loc=<0 0 0>, rot=<0 0 0 1>}
 dfrm_front_torso_end_up_r {parent=dfrm_mstr_front_torso_end_r, loc=<0 -2 0>, rot=<0 0 0 1>, scale=<0.25 0.25 0.25>}
 dfrm_mstr_back_torso_end_r {parent=spine_03, loc=<1.8111 -3.6274 7.7844>, rot=<0.1736 -0.9773 0.0081 -0.1215>}
 dfrm_back_torso_end_r {parent=dfrm_mstr_back_torso_end_r, loc=<0 0 0>, rot=<0 0 0 1>}
 dfrm_back_torso_end_up_r {parent=dfrm_mstr_back_torso_end_r, loc=<0 -2 0>, rot=<0 0 0 1>, scale=<0.25 0.25 0.25>}
 thigh_l {parent=pelvis, loc=<-1.8716 -0.0109 -11.9936>, rot=<0.0012 0.0572 0.0155 -0.9982>}
 calf_l {parent=thigh_l, loc=<-41.9178 -0.2246 0.0926>, rot=<0.0011 -0.016 -0.0109 -0.9998>}
 foot_l {parent=calf_l, loc=<-38.7057 1.2438 0.2452>, rot=<0.0975 -0.0474 0.0317 -0.9936>}
 dfrm_calf_mstr_l {parent=calf_l, loc=<-28.9331 0.9298 0.1833>, rot=<-0.0022 -0.0023 -0.6957 -0.7184>}
 dfrm_calf_outer_l {parent=dfrm_calf_mstr_l, loc=<9.9328 2.044 -7.2193>, rot=<0.0049 0.0002 0.0253 -0.9997>}
 dfrm_calf_inner_l {parent=dfrm_calf_mstr_l, loc=<9.9313 2.1889 7.4222>, rot=<0.0049 0.0002 0.0253 -0.9997>}
 dfrm_knee_mstr_l {parent=calf_l, loc=<0 0 0>, rot=<0 0 0 1>}
 dfrm_knee_l {parent=dfrm_knee_mstr_l, loc=<0 0 0>, rot=<0 0 0 1>}
 dfrm_thigh_aim_l {parent=dfrm_knee_l, loc=<12.7527 -2.6762 0.1885>, rot=<0 0 0 1>}
 dfrm_thigh_up_l {parent=thigh_l, loc=<0 5 0>, rot=<0 0 0 1>, scale=<0.25 0.25 0.25>}
 thigh_r {parent=pelvis, loc=<-1.8716 -0.0109 11.9936>, rot=<0.0572 -0.0012 0.9982 0.0155>}
 calf_r {parent=thigh_r, loc=<41.9178 0.2246 -0.0926>, rot=<0.0006 -0.0161 -0.0391 -0.9991>}
 foot_r {parent=calf_r, loc=<38.7057 -1.2438 -0.2452>, rot=<0.0975 -0.0474 0.0317 -0.9936>}
 dfrm_calf_mstr_r {parent=calf_r, loc=<28.9332 -0.9298 -0.1833>, rot=<-0.0022 -0.0023 -0.6957 -0.7184>}
 dfrm_calf_outer_r {parent=dfrm_calf_mstr_r, loc=<-9.9328 -2.044 7.2193>, rot=<0.0049 0.0002 0.0253 -0.9997>}
 dfrm_calf_inner_r {parent=dfrm_calf_mstr_r, loc=<-9.9313 -2.1888 -7.4222>, rot=<0.0049 0.0002 0.0253 -0.9997>}
 dfrm_knee_mstr_r {parent=calf_r, loc=<0 0 0>, rot=<0 0 0.0282 -0.9996>}
 dfrm_knee_r {parent=dfrm_knee_mstr_r, loc=<0 0 0>, rot=<0 0 0 1>}
 dfrm_thigh_aim_r {parent=dfrm_knee_r, loc=<-12.7527 2.6762 -0.1885>, rot=<0 0 0 1>}
 dfrm_thigh_up_r {parent=thigh_r, loc=<0 -5 0>, rot=<0 0 0 1>, scale=<0.25 0.25 0.25>}
 dfrm_hip_plate_l {parent=pelvis, loc=<8.4666 -3.605 -16.5963>, rot=<0.0998 -0.6274 -0.08 -0.7681>}
 dfrm_hip_plate_r {parent=pelvis, loc=<7.1863 -2.2049 16.2626>, rot=<-0.6284 0.0113 0.7777 0.0139>}
 dfrm_mstr_front_torso_base_l {parent=pelvis, loc=<9.6763 6.224 -5.9127>, rot=<0.9864 -0.1148 -0.1173 0.0094>}
 dfrm_front_torso_base_l {parent=dfrm_mstr_front_torso_base_l, loc=<0 0 0>, rot=<0 0 0 1>}
 dfrm_front_torso_mid_l {parent=dfrm_front_torso_base_l, loc=<8.2965 0.0531 -0.0072>, rot=<0 0 0 1>}
 dfrm_front_torso_base_up_l {parent=dfrm_mstr_front_torso_base_l, loc=<0 -2 0>, rot=<0 0 0 1>, scale=<0.25 0.25 0.25>}
 dfrm_mstr_back_torso_base_l {parent=pelvis, loc=<10.5562 -8.1621 -4.9701>, rot=<-0.0168 -0.1215 -0.101 -0.9873>}
 dfrm_back_torso_base_l {parent=dfrm_mstr_back_torso_base_l, loc=<0 0 0>, rot=<0 0 0 1>}
 dfrm_back_torso_mid_l {parent=dfrm_back_torso_base_l, loc=<4.7165 0 0>, rot=<0 0 0 1>}
 dfrm_back_torso_base_up_l {parent=dfrm_mstr_back_torso_base_l, loc=<0 -2 0>, rot=<0 0 0 1>, scale=<0.25 0.25 0.25>}
 dfrm_mstr_front_torso_base_r {parent=pelvis, loc=<9.6763 6.224 5.9127>, rot=<0.9864 -0.1148 0.1173 -0.0094>}
 dfrm_front_torso_base_r {parent=dfrm_mstr_front_torso_base_r, loc=<0 0 0>, rot=<0 0.0004 -0.0032 -1>}
 dfrm_front_torso_mid_r {parent=dfrm_front_torso_base_r, loc=<8.2967 0 0>, rot=<0 0.0004 0 -1>}
 dfrm_front_torso_base_up_r {parent=dfrm_mstr_front_torso_base_r, loc=<0 -2 0>, rot=<0 0 0 1>, scale=<0.25 0.25 0.25>}
 dfrm_mstr_back_torso_base_r {parent=pelvis, loc=<10.5566 -8.1621 4.9701>, rot=<0.0168 0.1215 -0.101 -0.9873>}
 dfrm_back_torso_base_r {parent=dfrm_mstr_back_torso_base_r, loc=<0 0 0>, rot=<0 0 0 1>}
 dfrm_back_torso_mid_r {parent=dfrm_back_torso_base_r, loc=<4.7166 0 0>, rot=<0 0 0 1>}
 dfrm_back_torso_base_up_r {parent=dfrm_mstr_back_torso_base_r, loc=<0 -2 0>, rot=<0 0 0 1>, scale=<0.25 0.25 0.25>}
 dfrm_thigh_l {parent=pelvis, loc=<-1.8716 -0.0109 -11.9936>, rot=<0.0012 0.0572 0.0155 -0.9982>}
 dfrm_thigh_r {parent=pelvis, loc=<-1.8716 -0.0109 11.9936>, rot=<0.0572 -0.0012 0.9982 0.0155>}
]
variables [
 current_parachute_value {type=double, default=0.35}
 delta_ref_torso_rotate_forward_alpha {type=double, default=0}
 delta_ref_torso_twist_right_alpha {type=double, default=0}
 get_is_parachute_opened {type=bool, default=False}
 get_is_skydive_floating {type=bool, default=False}
 get_is_skydiving {type=bool, default=False}
 spr_clavicle_l_up_alpha {type=double, default=0}
 spr_clavicle_r_up_alpha {type=double, default=0}
 spr_foot_l_rotate_in_alpha {type=double, default=0}
 spr_foot_l_rotate_out_alpha {type=double, default=0}
 spr_foot_l_rotate_up_alpha {type=double, default=0}
 spr_foot_r_rotate_in_alpha {type=double, default=0}
 spr_foot_r_rotate_out_alpha {type=double, default=0}
 spr_foot_r_rotate_up_alpha {type=double, default=0}
 spr_thigh_l_rotate_out_alpha {type=double, default=0}
 spr_thigh_l_twist_out_alpha {type=double, default=0}
 spr_thigh_l_up_alpha {type=double, default=0}
 spr_thigh_r_rotate_out_alpha {type=double, default=0}
 spr_thigh_r_twist_out_alpha {type=double, default=0}
 spr_thigh_r_up_alpha {type=double, default=0}
 spr_upperarm_r_ads_up_alpha {type=double, default=0}
 spr_upperarm_r_rot_in_alpha {type=double, default=0}
]
