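# Post-process style animation Blueprint for the large male HighMotion player body cosmetic, layering Control Rig adjustments onto the character's pose.
# It copies the pose from the attached parent mesh into M_LRG_HighMotion_CtrlRig, along with skydiving, floating and parachute flags.
# Each update, it converts spine_05's quaternion twist relative to pelvis, measured against the reference pose, into torso twist-right and rotate-forward alphas that drive the rig.
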AnimBlueprint /BRCosmetics/Characters/Player/Male/Large/Bodies/M_LRG_HighMotion/Meshes/M_LRG_HighMotion_AnimBP
target_skeleton=&/M_LRG_HighMotion_Skeleton
graphs [
 AnimGraph {nodes [
   n1 = Control Rig!(
    Source=Copy Pose From Mesh() {bUseAttachedParent=true}
    Alpha
    delta_ref_torso_twist_right_alpha:double=Get delta_ref_torso_twist_right_alpha() -> (delta_ref_torso_twist_right_alpha:double)
    delta_ref_torso_rotate_forward_alpha:double=Get delta_ref_torso_rotate_forward_alpha() -> (delta_ref_torso_rotate_forward_alpha:double)
    get_is_skydiving:bool=Get bIsSkyDiving()
    get_is_skydive_floating:bool=Get bIsSkydiveFloating()
    get_is_parachute_opened:bool=Get bIsParachuteOpened()
   ) -> (Pose) {
    alphaBoolBlend {blendOption=Linear}
    controlRigAssetReference {blueprintRigClass=&/M_LRG_HighMotion_CtrlRig.M_LRG_HighMotion_CtrlRig_C}
   }
   Output Pose(Result=@n1) {layerGroup=None}
  ]}
 EventGraph {nodes [
   Event BlueprintUpdateAnimation()
   n1 = Get skeletal_mesh() -> (skeletal_mesh:Skeletal Mesh Component Object Reference)
   > Set delta_ref_torso_twist_right_alpha(delta_ref_torso_twist_right_alpha:double=GetAlphaFromReferenceTransformDelta(
     SkeletalMeshComponent=@n1
     BoneName=spine_05
     BaseName=pelvis
     DeltaComponentType=QuaternionTwist
     InputRange=(X=-8,Y=-40)
     TwistAxis="1,0,0"
    )) -> (Output_Get:double)
   > Set delta_ref_torso_rotate_forward_alpha(delta_ref_torso_rotate_forward_alpha:double=GetAlphaFromReferenceTransformDelta(
     SkeletalMeshComponent=@n1
     BoneName=spine_05
     BaseName=pelvis
     DeltaComponentType=QuaternionTwist
     InputRange=(X=-11,Y=-38)
     TwistAxis="0,0,1"
    )) -> (Output_Get:double)
   Event BlueprintInitializeAnimation()
   > Set skeletal_mesh(skeletal_mesh:Skeletal Mesh Component Object Reference=GetOwnerSkeletalMeshComponent()) -> (Output_Get:Skeletal Mesh Component Object Reference)
  ]}
]
variables [
 delta_ref_torso_rotate_forward_alpha {type=number}
 delta_ref_torso_twist_right_alpha {type=number}
 skeletal_mesh {type=SkeletalMeshComponent}
]
