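# Control Rig for the M_LRG_HighMotion face accessory that drives four pairs of neck deform bone links per side between upper spine and head.
# Deform neck and head bones partially follow animated rotation, then spherical pose readers measure head poses.
# Back and tilt alphas weight corrective offsets on link endpoints, which then aim at each other.
# Look alphas go unused, and a False multiplier zeroes the crouch corrective.
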ControlRigBlueprint /BRCosmetics/Characters/Player/Male/Large/Bodies/M_LRG_HighMotion/Meshes/Parts/M_LRG_HighMotion_FaceAcc_CtrlRig
graphs [
 ContainedGraph {nodes [
   n1 = RigVMFunctionEntryNode()
   SequenceAggregate = &RigVMFunction_Sequence(ExecuteContext=@n1.ExecuteContext)
   SequenceAggregate_1 = &RigVMFunction_Sequence(ExecuteContext=@SequenceAggregate.B)
   SequenceAggregate_2 = &RigVMFunction_Sequence(ExecuteContext=@SequenceAggregate_1.B)
   SequenceAggregate_3 = &RigVMFunction_Sequence(ExecuteContext=@SequenceAggregate_2.B)
   RigVMFunctionReturnNode(A=@SequenceAggregate.A, B=@SequenceAggregate_1.A, C=@SequenceAggregate_2.A, D=@SequenceAggregate_3.A, E=@SequenceAggregate_3.B)
  ]}
 RigVMFunctionLibrary {links [], nodes []}
 RigVMModel {nodes [
   RigUnit_BeginExecution = &RigUnit_BeginExecution()
   SequenceAggregate = RigVMAggregateNode(ExecuteContext=@RigUnit_BeginExecution.ExecutePin)
   &RigUnit_RotationConstraintLocalSpaceOffset(
    AdvancedSettings=(RotationOrderForFilter=XZY)
    Child=(Type=Bone,Name="dfrm_neck_02")
    Filter=(bX=True,bY=True,bZ=True)
    Parents=((Item=(Type=Bone,Name="neck_02"),Weight=1))
    Weight=0.8
    ExecutePin=@SequenceAggregate.A
   )
   > &RigUnit_RotationConstraintLocalSpaceOffset(
    AdvancedSettings=(RotationOrderForFilter=XZY)
    Child=(Type=Bone,Name="dfrm_head")
    Filter=(bX=True,bY=True,bZ=True)
    Parents=((Item=(Type=Bone,Name="head"),Weight=1))
    Weight=0.8
   )
   n1 = &RigUnit_SphericalPoseReader(
    ActiveRegionScaleFactors=(PositiveWidth=0,NegativeWidth=1,PositiveHeight=1,NegativeHeight=1)
    ActiveRegionSize=0.1
    Debug=(bDrawDebug=false,DebugScale=25,DebugSegments=20,DebugThickness=0.25)
    DriverAxis=(X=1,Y=0,Z=0)
    DriverItem=(Type=Bone,Name="dfrm_head")
    FalloffRegionScaleFactors=(PositiveWidth=0,NegativeWidth=1,PositiveHeight=1,NegativeHeight=1)
    FalloffSize=0.49
    OptionalParentItem=(Type=Bone)
    RotationOffset=(X=90,Y=0,Z=-180)
    ExecutePin=@SequenceAggregate.B
   )
   > RigVMVariableNode(Value=@n1.OutputParam) {variable=spr_dfrm_head_back_alpha}
   > n2 = &RigUnit_SphericalPoseReader(
    ActiveRegionScaleFactors=(PositiveWidth=1,NegativeWidth=0,PositiveHeight=1,NegativeHeight=1)
    ActiveRegionSize=0.1
    Debug=(bDrawDebug=false,DebugScale=25,DebugSegments=20,DebugThickness=0.25)
    DriverAxis=(X=1,Y=0,Z=0)
    DriverItem=(Type=Bone,Name="dfrm_head")
    FalloffRegionScaleFactors=(PositiveWidth=1,NegativeWidth=0,PositiveHeight=1,NegativeHeight=1)
    FalloffSize=0.49
    OptionalParentItem=(Type=Bone)
    RotationOffset=(X=0,Y=0,Z=0)
   )
   > RigVMVariableNode(Value=@n2.OutputParam) {variable=spr_dfrm_head_tilt_right_alpha}
   > n3 = &RigUnit_SphericalPoseReader(
    ActiveRegionScaleFactors=(PositiveWidth=0,NegativeWidth=1,PositiveHeight=1,NegativeHeight=1)
    ActiveRegionSize=0.1
    Debug=(bDrawDebug=false,DebugScale=25,DebugSegments=20,DebugThickness=0.25)
    DriverAxis=(X=1,Y=0,Z=0)
    DriverItem=(Type=Bone,Name="dfrm_head")
    FalloffRegionScaleFactors=(PositiveWidth=0,NegativeWidth=1,PositiveHeight=1,NegativeHeight=1)
    FalloffSize=0.49
    OptionalParentItem=(Type=Bone)
    RotationOffset=(X=0,Y=180,Z=0)
   )
   > RigVMVariableNode(Value=@n3.OutputParam) {variable=spr_dfrm_head_tilt_left_alpha}
   > n4 = &RigUnit_SphericalPoseReader(
    ActiveRegionScaleFactors=(PositiveWidth=1,NegativeWidth=1,PositiveHeight=1,NegativeHeight=0)
    ActiveRegionSize=0.1
    Debug=(bDrawDebug=false,DebugScale=25,DebugSegments=20,DebugThickness=0.25)
    DriverAxis=(X=0,Y=1,Z=0)
    DriverItem=(Type=Bone,Name="dfrm_head")
    FalloffRegionScaleFactors=(PositiveWidth=1,NegativeWidth=1,PositiveHeight=1,NegativeHeight=0)
    FalloffSize=0.49
    OptionalParentItem=(Type=Bone)
    RotationOffset=(X=0,Y=0,Z=0)
   )
   > RigVMVariableNode(Value=@n4.OutputParam) {variable=spr_dfrm_head_look_right_alpha}
   > n5 = &RigUnit_SphericalPoseReader(
    ActiveRegionScaleFactors=(PositiveWidth=1,NegativeWidth=1,PositiveHeight=0,NegativeHeight=1)
    ActiveRegionSize=0.1
    Debug=(bDrawDebug=false,DebugScale=25,DebugSegments=20,DebugThickness=0.25)
    DriverAxis=(X=0,Y=1,Z=0)
    DriverItem=(Type=Bone,Name="dfrm_head")
    FalloffRegionScaleFactors=(PositiveWidth=1,NegativeWidth=1,PositiveHeight=0,NegativeHeight=1)
    FalloffSize=0.49
    OptionalParentItem=(Type=Bone)
    RotationOffset=(X=180,Y=0,Z=0)
   )
   > RigVMVariableNode(Value=@n5.OutputParam) {variable=spr_dfrm_head_look_left_alpha}
   > n6 = &RigUnit_SphericalPoseReader(
    ActiveRegionScaleFactors=(PositiveWidth=3,NegativeWidth=0.5,PositiveHeight=3,NegativeHeight=3)
    ActiveRegionSize=0.05
    Debug=(bDrawDebug=false,DebugScale=25,DebugSegments=20,DebugThickness=0.25)
    DriverAxis=(X=0,Y=1,Z=0)
    DriverItem=(Type=Bone,Name="head")
    FalloffRegionScaleFactors=(PositiveWidth=1,NegativeWidth=1,PositiveHeight=1,NegativeHeight=1)
    FalloffSize=0.15
    OptionalParentItem=(Type=Bone)
    RotationOffset=(X=100,Y=0,Z=-25)
   )
   > RigVMVariableNode(Value=@n6.OutputParam) {variable=spr_dfrm_head_tilt_up_right_crouch_alpha}
   SequenceAggregate_2 = &RigVMFunction_Sequence(ExecuteContext=@SequenceAggregate.E)
   /Script/ControlRig.RigUnit_AimItem(
    DebugSettings=(Scale=10,WorldOffset=(Rotation=(X=0,Y=0,Z=0,W=1),Translation=(X=0,Y=0,Z=0),Scale3D=(X=1,Y=1,Z=1)))
    Item=(Type=Bone,Name="dfrm_aim_base_neck_01_l")
    Primary=(Weight=1,Axis=(X=1,Y=0,Z=0),Target=(X=0,Y=0,Z=0),Kind=Location,Space=(Type=Bone,Name="dfrm_mstr_end_neck_01_l"))
    Secondary=(Weight=1,Axis=(X=0,Y=1,Z=0),Target=(X=1,Y=0,Z=0),Kind=Location,Space=(Type=Bone,Name="dfrm_aim_up_base_neck_01_l"))
    Weight=1
    ExecutePin=@SequenceAggregate_2.A
   ) {is=struct RigVMUnitNode}
   > &RigUnit_AimItem(
    DebugSettings=(Scale=10,WorldOffset=(Rotation=(X=0,Y=0,Z=0,W=1),Translation=(X=0,Y=0,Z=0),Scale3D=(X=1,Y=1,Z=1)))
    Item=(Type=Bone,Name="dfrm_aim_end_neck_01_l")
    Primary=(Weight=1,Axis=(X=1,Y=0,Z=0),Target=(X=0,Y=0,Z=0),Kind=Location,Space=(Type=Bone,Name="dfrm_mstr_base_neck_01_l"))
    Secondary=(Weight=1,Axis=(X=0,Y=1,Z=0),Target=(X=1,Y=0,Z=0),Kind=Location,Space=(Type=Bone,Name="dfrm_aim_up_end_neck_01_l"))
    Weight=1
   )
   > &RigUnit_AimItem(
    DebugSettings=(Scale=10,WorldOffset=(Rotation=(X=0,Y=0,Z=0,W=1),Translation=(X=0,Y=0,Z=0),Scale3D=(X=1,Y=1,Z=1)))
    Item=(Type=Bone,Name="dfrm_aim_base_neck_02_l")
    Primary=(Weight=1,Axis=(X=1,Y=0,Z=0),Target=(X=0,Y=0,Z=0),Kind=Location,Space=(Type=Bone,Name="dfrm_mstr_end_neck_02_l"))
    Secondary=(Weight=1,Axis=(X=0,Y=1,Z=0),Target=(X=1,Y=0,Z=0),Kind=Location,Space=(Type=Bone,Name="dfrm_aim_up_base_neck_02_l"))
    Weight=1
   )
   > &RigUnit_AimItem(
    DebugSettings=(Scale=10,WorldOffset=(Rotation=(X=0,Y=0,Z=0,W=1),Translation=(X=0,Y=0,Z=0),Scale3D=(X=1,Y=1,Z=1)))
    Item=(Type=Bone,Name="dfrm_aim_end_neck_02_l")
    Primary=(Weight=1,Axis=(X=1,Y=0,Z=0),Target=(X=0,Y=0,Z=0),Kind=Location,Space=(Type=Bone,Name="dfrm_mstr_base_neck_02_l"))
    Secondary=(Weight=1,Axis=(X=0,Y=1,Z=0),Target=(X=1,Y=0,Z=0),Kind=Location,Space=(Type=Bone,Name="dfrm_aim_up_end_neck_02_l"))
    Weight=1
   )
   > &RigUnit_AimItem(
    DebugSettings=(Scale=10,WorldOffset=(Rotation=(X=0,Y=0,Z=0,W=1),Translation=(X=0,Y=0,Z=0),Scale3D=(X=1,Y=1,Z=1)))
    Item=(Type=Bone,Name="dfrm_aim_base_neck_03_l")
    Primary=(Weight=1,Axis=(X=1,Y=0,Z=0),Target=(X=0,Y=0,Z=0),Kind=Location,Space=(Type=Bone,Name="dfrm_mstr_end_neck_03_l"))
    Secondary=(Weight=1,Axis=(X=0,Y=1,Z=0),Target=(X=1,Y=0,Z=0),Kind=Location,Space=(Type=Bone,Name="dfrm_aim_up_base_neck_03_l"))
    Weight=1
   )
   > &RigUnit_AimItem(
    DebugSettings=(Scale=10,WorldOffset=(Rotation=(X=0,Y=0,Z=0,W=1),Translation=(X=0,Y=0,Z=0),Scale3D=(X=1,Y=1,Z=1)))
    Item=(Type=Bone,Name="dfrm_aim_end_neck_03_l")
    Primary=(Weight=1,Axis=(X=1,Y=0,Z=0),Target=(X=0,Y=0,Z=0),Kind=Location,Space=(Type=Bone,Name="dfrm_mstr_base_neck_03_l"))
    Secondary=(Weight=1,Axis=(X=0,Y=1,Z=0),Target=(X=1,Y=0,Z=0),Kind=Location,Space=(Type=Bone,Name="dfrm_aim_up_end_neck_03_l"))
    Weight=1
   )
   > &RigUnit_AimItem(
    DebugSettings=(Scale=10,WorldOffset=(Rotation=(X=0,Y=0,Z=0,W=1),Translation=(X=0,Y=0,Z=0),Scale3D=(X=1,Y=1,Z=1)))
    Item=(Type=Bone,Name="dfrm_aim_base_neck_04_l")
    Primary=(Weight=1,Axis=(X=1,Y=0,Z=0),Target=(X=0,Y=0,Z=0),Kind=Location,Space=(Type=Bone,Name="dfrm_mstr_end_neck_04_l"))
    Secondary=(Weight=1,Axis=(X=0,Y=1,Z=0),Target=(X=1,Y=0,Z=0),Kind=Location,Space=(Type=Bone,Name="dfrm_aim_up_base_neck_04_l"))
    Weight=1
   )
   > &RigUnit_AimItem(
    DebugSettings=(Scale=10,WorldOffset=(Rotation=(X=0,Y=0,Z=0,W=1),Translation=(X=0,Y=0,Z=0),Scale3D=(X=1,Y=1,Z=1)))
    Item=(Type=Bone,Name="dfrm_aim_end_neck_04_l")
    Primary=(Weight=1,Axis=(X=1,Y=0,Z=0),Target=(X=0,Y=0,Z=0),Kind=Location,Space=(Type=Bone,Name="dfrm_mstr_base_neck_04_l"))
    Secondary=(Weight=1,Axis=(X=0,Y=1,Z=0),Target=(X=1,Y=0,Z=0),Kind=Location,Space=(Type=Bone,Name="dfrm_aim_up_end_neck_04_l"))
    Weight=1
   )
   &RigUnit_AimItem(
    DebugSettings=(Scale=10,WorldOffset=(Rotation=(X=0,Y=0,Z=0,W=1),Translation=(X=0,Y=0,Z=0),Scale3D=(X=1,Y=1,Z=1)))
    Item=(Type=Bone,Name="dfrm_aim_base_neck_01_r")
    Primary=(Weight=1,Axis=(X=1,Y=0,Z=0),Target=(X=0,Y=0,Z=0),Kind=Location,Space=(Type=Bone,Name="dfrm_mstr_end_neck_01_r"))
    Secondary=(Weight=1,Axis=(X=0,Y=1,Z=0),Target=(X=1,Y=0,Z=0),Kind=Location,Space=(Type=Bone,Name="dfrm_aim_up_base_neck_01_r"))
    Weight=1
    ExecutePin=@SequenceAggregate_2.B
   )
   > &RigUnit_AimItem(
    DebugSettings=(Scale=10,WorldOffset=(Rotation=(X=0,Y=0,Z=0,W=1),Translation=(X=0,Y=0,Z=0),Scale3D=(X=1,Y=1,Z=1)))
    Item=(Type=Bone,Name="dfrm_aim_end_neck_01_r")
    Primary=(Weight=1,Axis=(X=1,Y=0,Z=0),Target=(X=0,Y=0,Z=0),Kind=Location,Space=(Type=Bone,Name="dfrm_aim_base_neck_01_r"))
    Secondary=(Weight=1,Axis=(X=0,Y=1,Z=0),Target=(X=1,Y=0,Z=0),Kind=Location,Space=(Type=Bone,Name="dfrm_aim_up_end_neck_01_r"))
    Weight=1
   )
   > &RigUnit_AimItem(
    DebugSettings=(Scale=10,WorldOffset=(Rotation=(X=0,Y=0,Z=0,W=1),Translation=(X=0,Y=0,Z=0),Scale3D=(X=1,Y=1,Z=1)))
    Item=(Type=Bone,Name="dfrm_aim_base_neck_02_r")
    Primary=(Weight=1,Axis=(X=1,Y=0,Z=0),Target=(X=0,Y=0,Z=0),Kind=Location,Space=(Type=Bone,Name="dfrm_mstr_end_neck_02_r"))
    Secondary=(Weight=1,Axis=(X=0,Y=1,Z=0),Target=(X=1,Y=0,Z=0),Kind=Location,Space=(Type=Bone,Name="dfrm_aim_up_base_neck_02_r"))
    Weight=1
   )
   > &RigUnit_AimItem(
    DebugSettings=(Scale=10,WorldOffset=(Rotation=(X=0,Y=0,Z=0,W=1),Translation=(X=0,Y=0,Z=0),Scale3D=(X=1,Y=1,Z=1)))
    Item=(Type=Bone,Name="dfrm_aim_end_neck_02_r")
    Primary=(Weight=1,Axis=(X=1,Y=0,Z=0),Target=(X=0,Y=0,Z=0),Kind=Location,Space=(Type=Bone,Name="dfrm_mstr_base_neck_02_r"))
    Secondary=(Weight=1,Axis=(X=0,Y=1,Z=0),Target=(X=1,Y=0,Z=0),Kind=Location,Space=(Type=Bone,Name="dfrm_aim_up_end_neck_02_r"))
    Weight=1
   )
   > &RigUnit_AimItem(
    DebugSettings=(Scale=10,WorldOffset=(Rotation=(X=0,Y=0,Z=0,W=1),Translation=(X=0,Y=0,Z=0),Scale3D=(X=1,Y=1,Z=1)))
    Item=(Type=Bone,Name="dfrm_aim_base_neck_03_r")
    Primary=(Weight=1,Axis=(X=1,Y=0,Z=0),Target=(X=0,Y=0,Z=0),Kind=Location,Space=(Type=Bone,Name="dfrm_mstr_end_neck_03_r"))
    Secondary=(Weight=1,Axis=(X=0,Y=1,Z=0),Target=(X=1,Y=0,Z=0),Kind=Location,Space=(Type=Bone,Name="dfrm_aim_up_base_neck_03_r"))
    Weight=1
   )
   > &RigUnit_AimItem(
    DebugSettings=(Scale=10,WorldOffset=(Rotation=(X=0,Y=0,Z=0,W=1),Translation=(X=0,Y=0,Z=0),Scale3D=(X=1,Y=1,Z=1)))
    Item=(Type=Bone,Name="dfrm_aim_end_neck_03_r")
    Primary=(Weight=1,Axis=(X=1,Y=0,Z=0),Target=(X=0,Y=0,Z=0),Kind=Location,Space=(Type=Bone,Name="dfrm_mstr_base_neck_03_r"))
    Secondary=(Weight=1,Axis=(X=0,Y=1,Z=0),Target=(X=1,Y=0,Z=0),Kind=Location,Space=(Type=Bone,Name="dfrm_aim_up_end_neck_03_r"))
    Weight=1
   )
   > &RigUnit_AimItem(
    DebugSettings=(Scale=10,WorldOffset=(Rotation=(X=0,Y=0,Z=0,W=1),Translation=(X=0,Y=0,Z=0),Scale3D=(X=1,Y=1,Z=1)))
    Item=(Type=Bone,Name="dfrm_aim_base_neck_04_r")
    Primary=(Weight=1,Axis=(X=1,Y=0,Z=0),Target=(X=0,Y=0,Z=0),Kind=Location,Space=(Type=Bone,Name="dfrm_mstr_end_neck_04_r"))
    Secondary=(Weight=1,Axis=(X=0,Y=1,Z=0),Target=(X=1,Y=0,Z=0),Kind=Location,Space=(Type=Bone,Name="dfrm_aim_up_base_neck_04_r"))
    Weight=1
   )
   > &RigUnit_AimItem(
    DebugSettings=(Scale=10,WorldOffset=(Rotation=(X=0,Y=0,Z=0,W=1),Translation=(X=0,Y=0,Z=0),Scale3D=(X=1,Y=1,Z=1)))
    Item=(Type=Bone,Name="dfrm_aim_end_neck_04_r")
    Primary=(Weight=1,Axis=(X=1,Y=0,Z=0),Target=(X=0,Y=0,Z=0),Kind=Location,Space=(Type=Bone,Name="dfrm_mstr_base_neck_04_r"))
    Secondary=(Weight=1,Axis=(X=0,Y=1,Z=0),Target=(X=1,Y=0,Z=0),Kind=Location,Space=(Type=Bone,Name="dfrm_aim_up_end_neck_04_r"))
    Weight=1
   )
   Cast = &RigVMFunction_MathBoolToFloat(Value=False)
   Multiply = &RigVMFunction_MathDoubleMul(A=@Cast.Result, B=RigVMVariableNode() {variable=spr_dfrm_head_tilt_up_right_crouch_alpha}.Value)
   &RigUnit_ModifyTransforms(
    ItemToModify=((Item=(Type=Bone,Name="dfrm_mstr_end_neck_01_l"),Transform=(Rotation=(X=0,Y=0,Z=0,W=1),Translation=(X=0,Y=0.5,Z=-0.5),Scale3D=(X=1,Y=1,Z=1))))
    Weight=@Multiply.Result
    ExecutePin=@SequenceAggregate.D
   )
   > &RigUnit_ModifyTransforms(
    ItemToModify=((Item=(Type=Bone,Name="dfrm_mstr_base_neck_01_l"),Transform=(Rotation=(X=0,Y=0,Z=0,W=1),Translation=(X=0.5,Y=3.5,Z=0.5),Scale3D=(X=1,Y=1,Z=1))))
    Weight=@Multiply.Result
   )
   > &RigUnit_ModifyTransforms(
    ItemToModify=((Item=(Type=Bone,Name="dfrm_mstr_end_neck_02_l"),Transform=(Rotation=(X=0,Y=0,Z=0,W=1),Translation=(X=0,Y=1,Z=-1),Scale3D=(X=1,Y=1,Z=1))))
    Weight=@Multiply.Result
   )
   > &RigUnit_ModifyTransforms(
    ItemToModify=((Item=(Type=Bone,Name="dfrm_mstr_base_neck_02_l"),Transform=(Rotation=(X=0,Y=0,Z=0,W=1),Translation=(X=1,Y=3.5,Z=1.5),Scale3D=(X=1,Y=1,Z=1))))
    Weight=@Multiply.Result
   )
   > &RigUnit_ModifyTransforms(
    ItemToModify=((Item=(Type=Bone,Name="dfrm_mstr_end_neck_03_l"),Transform=(Rotation=(X=0,Y=0,Z=0,W=1),Translation=(X=2,Y=-0.5,Z=0),Scale3D=(X=1,Y=1,Z=1))))
    Weight=@Multiply.Result
   )
   > &RigUnit_ModifyTransforms(
    ItemToModify=((Item=(Type=Bone,Name="dfrm_mstr_base_neck_03_l"),Transform=(Rotation=(X=0,Y=0,Z=0,W=1),Translation=(X=-1,Y=-2,Z=-1),Scale3D=(X=1,Y=1,Z=1))))
    Weight=@Multiply.Result
   )
   n7 = RigVMVariableNode() {variable=spr_dfrm_head_back_alpha}
   &RigUnit_ModifyTransforms(
    ItemToModify=((Item=(Type=Bone,Name="dfrm_mstr_end_neck_01_l"),Transform=(Rotation=(X=0,Y=0,Z=0,W=1),Translation=(X=0,Y=0,Z=-1.5),Scale3D=(X=1,Y=1,Z=1))))
    Weight=@n7.Value
    ExecutePin=@SequenceAggregate.C
   )
   > &RigUnit_ModifyTransforms(
    ItemToModify=((Item=(Type=Bone,Name="dfrm_mstr_end_neck_01_r"),Transform=(Rotation=(X=0,Y=0,Z=0,W=1),Translation=(X=0,Y=0,Z=1.5),Scale3D=(X=1,Y=1,Z=1))))
    Weight=@n7.Value
   )
   n8 = RigVMVariableNode() {variable=spr_dfrm_head_tilt_right_alpha}
   > &RigUnit_ModifyTransforms(
    ItemToModify=((Item=(Type=Bone,Name="dfrm_mstr_base_neck_01_l"),Transform=(Rotation=(X=0,Y=0,Z=0,W=1),Translation=(X=0,Y=2,Z=3),Scale3D=(X=1,Y=1,Z=1))))
    Weight=@n8.Value
   )
   > &RigUnit_ModifyTransforms(
    ItemToModify=((Item=(Type=Bone,Name="dfrm_mstr_base_neck_02_l"),Transform=(Rotation=(X=0,Y=0,Z=0,W=1),Translation=(X=0,Y=2,Z=3),Scale3D=(X=1,Y=1,Z=1))))
    Weight=@n8.Value
   )
   > &RigUnit_ModifyTransforms(
    ItemToModify=((Item=(Type=Bone,Name="dfrm_mstr_end_neck_01_l"),Transform=(Rotation=(X=0,Y=0,Z=0,W=1),Translation=(X=4,Y=0,Z=-4),Scale3D=(X=1,Y=1,Z=1))))
    Weight=@n8.Value
   )
   > &RigUnit_ModifyTransforms(
    ItemToModify=((Item=(Type=Bone,Name="dfrm_mstr_end_neck_02_l"),Transform=(Rotation=(X=0,Y=0,Z=0,W=1),Translation=(X=3,Y=0,Z=-3),Scale3D=(X=1,Y=1,Z=1))))
    Weight=@n8.Value
   )
   n9 = RigVMVariableNode() {variable=spr_dfrm_head_tilt_left_alpha}
   > &RigUnit_ModifyTransforms(
    ItemToModify=((Item=(Type=Bone,Name="dfrm_mstr_base_neck_01_r"),Transform=(Rotation=(X=0,Y=0,Z=0,W=1),Translation=(X=0,Y=2,Z=-3),Scale3D=(X=1,Y=1,Z=1))))
    Weight=@n9.Value
   )
   > &RigUnit_ModifyTransforms(
    ItemToModify=((Item=(Type=Bone,Name="dfrm_mstr_base_neck_02_r"),Transform=(Rotation=(X=0,Y=0,Z=0,W=1),Translation=(X=0,Y=2,Z=-3),Scale3D=(X=1,Y=1,Z=1))))
    Weight=@n9.Value
   )
   > &RigUnit_ModifyTransforms(
    ItemToModify=((Item=(Type=Bone,Name="dfrm_mstr_end_neck_01_r"),Transform=(Rotation=(X=0,Y=0,Z=0,W=1),Translation=(X=4,Y=0,Z=4),Scale3D=(X=1,Y=1,Z=1))))
    Weight=@n9.Value
   )
   > &RigUnit_ModifyTransforms(
    ItemToModify=((Item=(Type=Bone,Name="dfrm_mstr_end_neck_02_r"),Transform=(Rotation=(X=0,Y=0,Z=0,W=1),Translation=(X=3,Y=0,Z=3),Scale3D=(X=1,Y=1,Z=1))))
    Weight=@n9.Value
   )
  ]}
]
hierarchy [
 root {loc=<0 0 0>, rot=<0 0 0 1>}
 pelvis {parent=root, loc=<0 -0.7321 92.4218>, rot=<0.0127 0.707 -0.0127 -0.707>}
 spine_01 {parent=pelvis, loc=<5.6002 -0.5128 0>, rot=<0 0 0.034 -0.9994>}
 spine_02 {parent=spine_01, loc=<8.0753 0.0139 0>, rot=<0 0 0.0238 -0.9997>}
 spine_03 {parent=spine_02, loc=<7.0393 0 0>, rot=<0 0 0.0149 -0.9999>}
 spine_04 {parent=spine_03, loc=<8.1024 0 0>, rot=<0 0 -0.0132 -0.9999>}
 spine_05 {parent=spine_04, loc=<10.0146 0 0>, rot=<0 0 -0.0015 -1>}
 dfrm_neck_03 {parent=spine_05, loc=<7.1212 -1.7703 0>, rot=<0 0 -0.1864 -0.9825>}
 dfrm_neck_02 {parent=dfrm_neck_03, loc=<4.7347 0.3045 0>, rot=<0 0 0.0363 -0.9993>}
 dfrm_head {parent=dfrm_neck_02, loc=<5.7632 0.5552 0>, rot=<0 0 0.115 -0.9934>}
 dfrm_mstr_end_neck_01_l {parent=dfrm_head, loc=<1.4209 1.6765 -4.1902>, rot=<0.0466 -0.2073 0.966 0.1476>}
 dfrm_aim_end_neck_01_l {parent=dfrm_mstr_end_neck_01_l, loc=<0 0 0>, rot=<0 0 0 1>}
 dfrm_scale_end_neck_01_l {parent=dfrm_aim_end_neck_01_l, loc=<0 0 0>, rot=<0 0 0 1>}
 dfrm_aim_up_end_neck_01_l {parent=dfrm_mstr_end_neck_01_l, loc=<0 2 0>, rot=<0 0 0 1>, scale=<0.5 0.5 0.5>}
 dfrm_mstr_end_neck_02_l {parent=dfrm_head, loc=<2.1983 0.1808 -5.1915>, rot=<0.0743 0.0094 0.9903 0.1173>}
 dfrm_aim_end_neck_02_l {parent=dfrm_mstr_end_neck_02_l, loc=<0 0 0>, rot=<0 0 0 1>}
 dfrm_scale_end_neck_02_l {parent=dfrm_aim_end_neck_02_l, loc=<0 0 0>, rot=<0 0 0 1>}
 dfrm_aim_up_end_neck_02_l {parent=dfrm_mstr_end_neck_02_l, loc=<0 2 0>, rot=<0 0 0 1>, scale=<0.5 0.5 0.5>}
 dfrm_mstr_end_neck_03_l {parent=dfrm_head, loc=<4.8872 0.8397 -3.4544>, rot=<0.1038 -0.9928 0.0058 -0.0604>}
 dfrm_aim_end_neck_03_l {parent=dfrm_mstr_end_neck_03_l, loc=<0 0 0>, rot=<0 0 0 1>}
 dfrm_scale_end_neck_03_l {parent=dfrm_aim_end_neck_03_l, loc=<0 0 0>, rot=<0 0 0 1>}
 dfrm_aim_up_end_neck_03_l {parent=dfrm_mstr_end_neck_03_l, loc=<0 2 0>, rot=<0 0 0 1>, scale=<0.5 0.5 0.5>}
 dfrm_mstr_end_neck_04_l {parent=dfrm_head, loc=<4.1093 -1.3483 -2.9091>, rot=<0.1604 -0.9806 -0.1119 -0.0105>}
 dfrm_aim_end_neck_04_l {parent=dfrm_mstr_end_neck_04_l, loc=<0 0 0>, rot=<0 0 0 1>}
 dfrm_scale_end_neck_04_l {parent=dfrm_aim_end_neck_04_l, loc=<0 0 0>, rot=<0 0 0 1>}
 dfrm_aim_up_end_neck_04_l {parent=dfrm_mstr_end_neck_04_l, loc=<0 2 0>, rot=<0 0 0 1>, scale=<0.5 0.5 0.5>}
 dfrm_mstr_end_neck_01_r {parent=dfrm_head, loc=<1.4206 1.6765 4.1902>, rot=<-0.0466 0.2073 0.966 0.1476>}
 dfrm_aim_end_neck_01_r {parent=dfrm_mstr_end_neck_01_r, loc=<0 0 0>, rot=<0 0 0 1>}
 dfrm_scale_end_neck_01_r {parent=dfrm_aim_end_neck_01_r, loc=<0 0 0>, rot=<0 0 0 1>}
 dfrm_aim_up_end_neck_01_r {parent=dfrm_mstr_end_neck_01_r, loc=<0 2 0>, rot=<0 0 0 1>}
 dfrm_mstr_end_neck_02_r {parent=dfrm_head, loc=<2.198 0.1808 5.1915>, rot=<-0.0743 -0.0094 0.9903 0.1173>}
 dfrm_aim_end_neck_02_r {parent=dfrm_mstr_end_neck_02_r, loc=<0 0 0>, rot=<0 0 0 -1>}
 dfrm_scale_end_neck_02_r {parent=dfrm_aim_end_neck_02_r, loc=<0 0 0>, rot=<0 0 0 -1>}
 dfrm_aim_up_end_neck_02_r {parent=dfrm_mstr_end_neck_02_r, loc=<0 2 0>, rot=<0 0 0 1>, scale=<0.5 0.5 0.5>}
 dfrm_mstr_end_neck_03_r {parent=dfrm_head, loc=<4.8876 0.8397 3.4544>, rot=<0.1038 -0.9928 -0.0058 0.0604>}
 dfrm_aim_end_neck_03_r {parent=dfrm_mstr_end_neck_03_r, loc=<0 0 0>, rot=<0 0 0 -1>}
 dfrm_scale_end_neck_03_r {parent=dfrm_aim_end_neck_03_r, loc=<0 0 0>, rot=<0 0 0 -1>}
 dfrm_aim_up_end_neck_03_r {parent=dfrm_mstr_end_neck_03_r, loc=<0 2 0>, rot=<0 0 0 1>, scale=<0.5 0.5 0.5>}
 dfrm_mstr_end_neck_04_r {parent=dfrm_head, loc=<4.1097 -1.3483 2.9091>, rot=<0.1604 -0.9806 0.1119 0.0105>}
 dfrm_aim_end_neck_04_r {parent=dfrm_mstr_end_neck_04_r, loc=<0 0 0>, rot=<0 0 0 -1>}
 dfrm_scale_end_neck_04_r {parent=dfrm_aim_end_neck_04_r, loc=<0 0 0>, rot=<0 0 0 -1>}
 dfrm_aim_up_end_neck_04_r {parent=dfrm_mstr_end_neck_04_r, loc=<0 2 0>, rot=<0 0 0 1>, scale=<0.5 0.5 0.5>}
 neck_01 {parent=spine_05, loc=<10.815 1.9196 0>, rot=<0 0 -0.1864 -0.9825>}
 neck_02 {parent=neck_01, loc=<4.7347 0.3045 0>, rot=<0 0 0.0363 -0.9993>}
 head {parent=neck_02, loc=<5.7632 0.5552 0>, rot=<0 0 0.115 -0.9934>}
 dfrm_mstr_base_neck_01_l {parent=spine_05, loc=<6.8656 6.5384 -2.4435>, rot=<0.9706 -0.1128 -0.0541 -0.2055>}
 dfrm_aim_base_neck_01_l {parent=dfrm_mstr_base_neck_01_l, loc=<0 0 0>, rot=<0 0 0 1>}
 dfrm_scale_base_neck_01_l {parent=dfrm_aim_base_neck_01_l, loc=<0 0 0>, rot=<0 0 0 1>}
 dfrm_aim_up_base_neck_01_l {parent=dfrm_mstr_base_neck_01_l, loc=<0 2 0>, rot=<0 0 0 1>, scale=<0.5 0.5 0.5>}
 dfrm_mstr_base_neck_02_l {parent=spine_05, loc=<5.4437 5.099 -3.1876>, rot=<0.9938 -0.0816 -0.0739 0.0121>}
 dfrm_aim_base_neck_02_l {parent=dfrm_mstr_base_neck_02_l, loc=<0 0 0>, rot=<0 0 0 1>}
 dfrm_scale_base_neck_02_l {parent=dfrm_aim_base_neck_02_l, loc=<0 0 0>, rot=<0 0 0 1>}
 dfrm_aim_up_base_neck_02_l {parent=dfrm_mstr_base_neck_02_l, loc=<0 2 0>, rot=<0 0 0 1>, scale=<0.5 0.5 0.5>}
 dfrm_mstr_base_neck_04_l {parent=spine_05, loc=<5.2946 -4.9902 -3.8676>, rot=<-0.1122 0.0065 -0.1956 -0.9742>}
 dfrm_aim_base_neck_04_l {parent=dfrm_mstr_base_neck_04_l, loc=<0 0 0>, rot=<0 0 0 1>}
 dfrm_scale_base_neck_04_l {parent=dfrm_aim_base_neck_04_l, loc=<0 0 0>, rot=<0 0 0 1>}
 dfrm_aim_up_base_neck_04_l {parent=dfrm_mstr_base_neck_04_l, loc=<0 2 0>, rot=<0 0 0 1>, scale=<0.5 0.5 0.5>}
 dfrm_mstr_base_neck_03_l {parent=spine_05, loc=<5.3982 -0.9961 -5.4658>, rot=<0.0036 0.0605 -0.1394 -0.9884>}
 dfrm_aim_base_neck_03_l {parent=dfrm_mstr_base_neck_03_l, loc=<0 0 0>, rot=<0 0 0 1>}
 dfrm_scale_base_neck_03_l {parent=dfrm_aim_base_neck_03_l, loc=<0 0 0>, rot=<0 0 0 1>}
 dfrm_aim_up_base_neck_03_l {parent=dfrm_mstr_base_neck_03_l, loc=<0 2 0>, rot=<0 0 0 1>, scale=<0.5 0.5 0.5>}
 dfrm_mstr_base_neck_01_r {parent=spine_05, loc=<6.8651 6.5383 2.4435>, rot=<0.9706 -0.1128 0.0541 0.2055>}
 dfrm_aim_base_neck_01_r {parent=dfrm_mstr_base_neck_01_r, loc=<0 0 0>, rot=<0 0 0 1>}
 dfrm_scale_base_neck_01_r {parent=dfrm_aim_base_neck_01_r, loc=<0 0 0>, rot=<0 0 0 1>}
 dfrm_aim_up_base_neck_01_r {parent=dfrm_mstr_base_neck_01_r, loc=<0 2 0>, rot=<0 0 0 1>, scale=<0.5 0.5 0.5>}
 dfrm_mstr_base_neck_02_r {parent=spine_05, loc=<5.4439 5.0991 3.1876>, rot=<0.9938 -0.0816 0.0739 -0.0121>}
 dfrm_aim_base_neck_02_r {parent=dfrm_mstr_base_neck_02_r, loc=<0 0 0>, rot=<0 0 0 1>}
 dfrm_scale_base_neck_02_r {parent=dfrm_aim_base_neck_02_r, loc=<0 0 0>, rot=<0 0 0 -1>}
 dfrm_aim_up_base_neck_02_r {parent=dfrm_mstr_base_neck_02_r, loc=<0 2 0>, rot=<0 0 0 1>, scale=<0.5 0.5 0.5>}
 dfrm_mstr_base_neck_03_r {parent=spine_05, loc=<5.3982 -0.9961 5.4658>, rot=<-0.0036 -0.0605 -0.1394 -0.9884>}
 dfrm_aim_base_neck_03_r {parent=dfrm_mstr_base_neck_03_r, loc=<0 0 0>, rot=<0 0 0 -1>}
 dfrm_scale_base_neck_03_r {parent=dfrm_aim_base_neck_03_r, loc=<0 0 0>, rot=<0 0 0 -1>}
 dfrm_aim_up_base_neck_03_r {parent=dfrm_mstr_base_neck_03_r, loc=<0 2 0>, rot=<0 0 0 1>, scale=<0.5 0.5 0.5>}
 dfrm_mstr_base_neck_04_r {parent=spine_05, loc=<5.2949 -4.9902 3.8676>, rot=<0.1122 -0.0065 -0.1956 -0.9742>}
 dfrm_aim_base_neck_04_r {parent=dfrm_mstr_base_neck_04_r, loc=<0 0 0>, rot=<0 0 0 1>}
 dfrm_scale_base_neck_04_r {parent=dfrm_aim_base_neck_04_r, loc=<0 0 0>, rot=<0 0 0 1>}
 dfrm_aim_up_base_neck_04_r {parent=dfrm_mstr_base_neck_04_r, loc=<0 2 0>, rot=<0 0 0 1>, scale=<0.5 0.5 0.5>}
]
variables [
 spr_dfrm_head_back_alpha {type=double, default=0}
 spr_dfrm_head_look_left_alpha {type=double, default=0}
 spr_dfrm_head_look_right_alpha {type=double, default=0}
 spr_dfrm_head_tilt_left_alpha {type=double, default=0}
 spr_dfrm_head_tilt_right_alpha {type=double, default=0}
 spr_dfrm_head_tilt_up_right_crouch_alpha {type=double, default=0}
]
